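# Parent material for the TrafficHat Boba female character cosmetic, supporting skeletal meshes and morph targets.
# The shared 2023 anime shader chain (Pre, Calculations, Post) builds material attributes, with the shading model set to Cloth.
# An IsHat switch blends base color and emissive with darkened grayscale versions, using a HairMask isolated from the hat SRM red channel.
# Static switches select Super Level tier 1–3 functions borrowed from RoyalAngst_Tote, or none.
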Material /BRCosmetics/Characters/Player/Female/Medium/Bodies/F_MED_TrafficHat_Boba/Materials/M_TrafficHat_Boba_Parent
graph {
 properties {
  bAutomaticallySetUsageInEditor=false
  bUseMaterialAttributes=true
  bUsedWithMeshDeformer=true
  bUsedWithMorphTargets=true
  bUsedWithSkeletalMesh=true
  bUsedWithStaticLighting=true
  pixelDepthOffsetMode=PDOM_Legacy
  shadingModel=MSM_FromMaterialExpression
  usedShadingModels=Default Lit | Clear Coat | Cloth
 }
 nodes [
  n1 = MaterialFunctionCall!(CustomUV) {materialFunction=/Game/Global/Materials/Character_Parent_2023/Anime_Parent_2023/MaterialFunctions/MFC_AnimeShader_2023_Pre}
  n2 = MaterialFunctionCall!(CustomLightVector, OverrideNormal_WS, STT=@n1.STT, Normal_Map=@n1.STT, DFL=@n1.STT, FX=@n1.STT) {materialFunction=/Game/Global/Materials/Character_Parent_2023/Anime_Parent_2023/MaterialFunctions/MFC_AnimeShader_2023_Calculations}
  n3 = SetMaterialAttributes!(
   MaterialAttributes=MaterialFunctionCall!(
    Shading_Gradient=@n2.Shading_Gradient
    Threshold_Gradient=@n2.Shading_Gradient
    Specular_Gradient=@n2.Shading_Gradient
    Sheen_Mask=@n2.Shading_Gradient
    Thickness=@n2.Shading_Gradient
    DFL_Mask=@n2.Shading_Gradient
    FX=@n2.Shading_Gradient
    Metallic_Gradient=@n2.Shading_Gradient
    RimLight_Gradient=@n2.Shading_Gradient
    NormalBuffer=@n2.Shading_Gradient
    --------
    CINE_LightSources=@n2.Shading_Gradient
    CINE_LightColor=@n2.Shading_Gradient
    ---------
    Color_Lit=@n1.Color_Lit
    Color_Shaded=@n1.Color_Lit
    SRM=@n1.Color_Lit
    Emissive=@n1.Color_Lit
    DFL_Color=@n1.Color_Lit
    Specular_Color=@n1.Color_Lit
    RimLight_Color=@n1.Color_Lit
   ) {materialFunction=/Game/Global/Materials/Character_Parent_2023/Anime_Parent_2023/MaterialFunctions/MFC_AnimeShader_2023_Post}.Result
   Shading Model=ShadingModel() {shadingModel=MSM_Cloth}
  ) {attributeSetTypes [#1]}
  n4 = GetMaterialAttributes(None=@n3) {attributeGetTypes [#2, #3]}
  n5 = Constant() {r=1}
  n6 = LinearInterpolate(A=Multiply(A=Desaturation(None=@n4.Base Color, Fraction=@n5)) {constB=0.2}, B=@n4.Base Color, Alpha=NamedRerouteUsage())
  n7 = Constant() {r=1}
  n8 = TextureSample() {
   samplerType=SAMPLERTYPE_Masks
   texture=/BRCosmetics/Characters/Player/Female/Medium/Bodies/F_MED_TrafficHat_Boba/Textures/T_TrafficHat_Boba_FaceAccHat_SRM
  }
  n9 = Constant() {r=0.6}
  n10 = LinearInterpolate(
   A=Multiply(A=Desaturation(None=@n4.Emissive Color, Fraction=@n7)) {constB=0.2}
   B=@n4.Emissive Color
   Alpha=NamedRerouteDeclaration(None=SphereMask(A=@n8.R, B=@n9) {attenuationRadius=0.05}) {name=HairMask, nodeColor=<rgba 0 1 0.4609 1>}
  )
  NamedRerouteDeclaration(None=SetMaterialAttributes!(
    MaterialAttributes=@n4.MaterialAttributes
    Base Color=StaticSwitchParameter(True=@n6, False=@n4.Base Color) {group=Skin FX, parameterName=IsHat}
    Emissive Color=StaticSwitchParameter(True=@n10, False=@n4.Emissive Color) {group=Skin FX, parameterName=IsHat}
   ) {attributeSetTypes [#2, #3]}) {name=PreSL_MA, nodeColor=<rgba 0 0.6016 1 1>}
  NamedRerouteDeclaration(None=ComponentMask(None=@n1.SRM) {b=true}) {name=SkinMask, nodeColor=<rgba 0 0.4141 1 1>}
  NamedRerouteDeclaration(None=@n3) {name=PreSL_MA_Output, nodeColor=<rgba 1 0.7031 0 1>}
  NamedRerouteDeclaration(None=@n4.Base Color) {name=BaseColor, nodeColor=<rgba 0 0.1094 1 1>}
  NamedRerouteDeclaration(None=@n1.FX) {name=FX, nodeColor=<rgba 0.0312 1 0 1>}
  n11 = MaterialFunctionCall!(In=NamedRerouteUsage()) {materialFunction=/BRCosmetics/Characters/Player/Female/Medium/Bodies/F_MED_RoyalAngst_Tote/Materials/MF_S34SL_T1b}
  n12 = GetMaterialAttributes(None=@n11.Result) {attributeGetTypes [#2]}
  n13 = Desaturation(None=NamedRerouteUsage(), Fraction=Constant() {r=1})
  SetMaterialAttributes!(
   MaterialAttributes=@n12.MaterialAttributes
   Base Color=LinearInterpolate(A=@n12.Base Color, B=LinearInterpolate(A=@n13, B=@n12.Base Color), Alpha=NamedRerouteUsage())
  ) {attributeSetTypes [#2]}
  n14 = MaterialFunctionCall!(In=NamedRerouteUsage()) {materialFunction=/BRCosmetics/Characters/Player/Female/Medium/Bodies/F_MED_RoyalAngst_Tote/Materials/MF_S34SL_T2}
  n15 = GetMaterialAttributes(None=@n14.Result) {attributeGetTypes [#2]}
  n16 = Desaturation(None=NamedRerouteUsage(), Fraction=Constant() {r=1})
  SetMaterialAttributes(
   MaterialAttributes=@n15.MaterialAttributes
   Base Color=LinearInterpolate(A=@n15.Base Color, B=LinearInterpolate(A=@n16, B=@n15.Base Color), Alpha=NamedRerouteUsage())
  ) {attributeSetTypes [#2]}
  n17 = MaterialFunctionCall!(In=NamedRerouteUsage()) {materialFunction=/BRCosmetics/Characters/Player/Female/Medium/Bodies/F_MED_RoyalAngst_Tote/Materials/MF_S34SL_T3}
  n18 = GetMaterialAttributes(None=@n17.Result) {attributeGetTypes [#2]}
  n19 = Desaturation(None=NamedRerouteUsage(), Fraction=Constant() {r=1})
  SetMaterialAttributes(
   MaterialAttributes=@n18.MaterialAttributes
   Base Color=LinearInterpolate(A=@n18.Base Color, B=LinearInterpolate(A=@n19, B=@n18.Base Color), Alpha=NamedRerouteUsage())
  ) {attributeSetTypes [#2]}
  n20 = NamedRerouteUsage()
  StaticSwitchParameter(
   True=@n11.Result
   False=StaticSwitchParameter(True=@n14.Result, False=StaticSwitchParameter(True=@n17.Result, False=@n20) {group=00_SuperLevels, parameterName=SL_Tier3}) {group=00_SuperLevels, parameterName=SL_Tier2}
  ) {group=00_SuperLevels, parameterName=SL_Tier1}
 ]
}
