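# Control Rig for the F_MED_TrafficHat_Boba player body that applies pose-driven corrections to its belt and suspender bones.
# A sequence runs a shared DeformRig function, then chains of spherical pose readers on thigh, upperarm, clavicle and spine bones.
# Reader outputs weight additive ModifyTransforms offsets on dyn_ bones.
# Many weights are multiplied or reduced by state variables like Crouching_moving, Sprint, Emote, Jumping, Skydiving and Parachuting, confining fixes to particular animations.
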ControlRigBlueprint /BRCosmetics/Characters/Player/Female/Medium/Bodies/F_MED_TrafficHat_Boba/Meshes/F_MED_TrafficHat_Boba_CtrlRig
graphs [
 RigVMFunctionLibrary {links [], nodes []}
 RigVMFunction_Sequence_1_ContainedGraph {nodes [
   n1 = RigVMFunctionEntryNode()
   RigVMFunction_Sequence = &RigVMFunction_Sequence(ExecuteContext=@n1.ExecuteContext)
   n2 = &RigVMFunction_Sequence(ExecuteContext=@RigVMFunction_Sequence.B)
   n3 = &RigVMFunction_Sequence(ExecuteContext=@n2.B)
   n4 = &RigVMFunction_Sequence(ExecuteContext=@n3.B)
   n5 = &RigVMFunction_Sequence(ExecuteContext=@n4.B)
   n6 = &RigVMFunction_Sequence(ExecuteContext=@n5.B)
   n7 = &RigVMFunction_Sequence(ExecuteContext=@n6.B)
   RigVMFunctionReturnNode(A=@RigVMFunction_Sequence.A, B=@n2.A, C=@n3.A, D=@n4.A, E=@n5.A, F=@n6.A, G=@n7.A, H=@n7.B)
  ]}
 RigVMModel {nodes [
   RigUnit_BeginExecution = &RigUnit_BeginExecution()
   RigVMFunction_Sequence = RigVMAggregateNode(ExecuteContext=@RigUnit_BeginExecution.ExecutePin)
   FN_DeformRig_FNC = RigVMFunctionReferenceNode(
    DeformRig_ Curve="(deformRig_l_upperarm_up_alpha_72_#1=False,deformRig_r_upperarm_up_alpha_73_#2=False,deformRig_l_upperarm_fr_alpha_74_#3=False,deformRig_r_upperarm_fr_alpha_75_CB51B8 ...(truncated)"
    DeformRig_Bool="(upperarm_1_#4=False,elbow_3_#5=True,wrist_7_#6=False,groin_8_#7=True,glute_17_#8=True,knee_9_40C408324CA2 ...(truncated)"
    DeformRig_Debug=(upperarm_debug_18_#4=False,wrist_debug_20_#6=False,lat_debug_26_#9=False,groinGlute_debug_27_#7=False)
    DeformRig_Value="(upperarm_up_18_#10=(Rotation=(X=0,Y=0,Z=0,W=1),Translation=(X=0,Y=0,Z=8),Scale3D=(X=0.7,Y=1,Z=1)),upperarm_fwd_fr_19_#11=(Rotation=(X=0,Y=0,Z=0,W=1),Translation=(X=0,Y=10,Z=0),Scale3D=(X=0.5,Y=1,Z= ...(truncated)"
    ExecuteContext=@RigVMFunction_Sequence.A
   )
   n1 = &RigUnit_SphericalPoseReader(
    ActiveRegionScaleFactors=(PositiveWidth=1,NegativeWidth=1,PositiveHeight=1,NegativeHeight=1)
    ActiveRegionSize=0.05
    Debug=(bDrawDebug=false,bDraw2D=False,bDrawLocalAxes=False,DebugScale=25,DebugSegments=20,DebugThickness=0.25)
    DriverAxis=(X=-1,Y=0,Z=0)
    DriverItem=(Type=Bone,Name="thigh_r")
    FalloffRegionScaleFactors=(PositiveWidth=1,NegativeWidth=1,PositiveHeight=1,NegativeHeight=1)
    FalloffSize=0.2
    FlipHeightScaling=False
    FlipWidthScaling=False
    OptionalParentItem=(Type=Bone,Name="None")
    RotationOffset=(X=90,Y=0,Z=2)
    ExecutePin=@RigVMFunction_Sequence.B
   )
   n2 = &RigVMFunction_MathDoubleMul(A=@n1.OutputParam, B=RigVMVariableNode() {variable=Crouching_moving}.Value)
   > &RigUnit_ModifyTransforms(
    ItemToModify=((Item=(Type=Bone,Name="dyn_belt_ring_01"),Transform=(Rotation=(X=-0.0552,Y=-0.1183,Z=-0.419,W=0.8986),Translation=(X=0,Y=3,Z=2),Scale3D=(X=1,Y=1,Z=1))))
    WeightMinimum=0
    Weight=@n2.Result
   )
   > &RigUnit_ModifyTransforms(
    ItemToModify=((Item=(Type=Bone,Name="dyn_belt_l_02"),Transform=(Rotation=(X=0,Y=0,Z=0,W=1),Translation=(X=0,Y=0,Z=0),Scale3D=(X=1,Y=1,Z=1))))
    WeightMinimum=0
    Weight=@n2.Result
   )
   > n3 = &RigUnit_SphericalPoseReader(
    ActiveRegionScaleFactors=(PositiveWidth=1,NegativeWidth=1,PositiveHeight=1,NegativeHeight=1)
    ActiveRegionSize=0.05
    Debug=(bDrawDebug=false,bDraw2D=False,bDrawLocalAxes=False,DebugScale=25,DebugSegments=20,DebugThickness=0.25)
    DriverAxis=(X=-1,Y=0,Z=0)
    DriverItem=(Type=Bone,Name="thigh_r")
    FalloffRegionScaleFactors=(PositiveWidth=1,NegativeWidth=1,PositiveHeight=1,NegativeHeight=1)
    FalloffSize=0.2
    FlipHeightScaling=False
    FlipWidthScaling=False
    OptionalParentItem=(Type=Bone,Name="None")
    RotationOffset=(X=110,Y=0,Z=55)
   )
   > &RigUnit_ModifyTransforms(
    ItemToModify=((Item=(Type=Bone,Name="dyn_belt_l_02"),Transform=(Rotation=(X=0,Y=-0.2588,Z=0,W=0.9659),Translation=(X=0,Y=0,Z=0),Scale3D=(X=1,Y=1,Z=1))))
    WeightMinimum=0
    Weight=&RigVMFunction_MathDoubleMul(A=@n3.OutputParam, B=RigVMVariableNode() {variable=Crouching_moving}.Value).Result
   )
   > &RigUnit_ModifyTransforms(
    ItemToModify=((Item=(Type=Bone,Name="dyn_belt_ring_01"),Transform=(Rotation=(X=0,Y=-0.1736,Z=0,W=0.9848),Translation=(X=0,Y=-2,Z=0),Scale3D=(X=1,Y=1,Z=1))))
    WeightMinimum=0
    Weight=@n3.OutputParam
   )
   > n4 = &RigUnit_SphericalPoseReader(
    ActiveRegionScaleFactors=(PositiveWidth=1,NegativeWidth=1,PositiveHeight=1,NegativeHeight=2)
    ActiveRegionSize=0.05
    Debug=(bDrawDebug=false,bDraw2D=False,bDrawLocalAxes=False,DebugScale=25,DebugSegments=20,DebugThickness=0.25)
    DriverAxis=(X=1,Y=0,Z=0)
    DriverItem=(Type=Bone,Name="thigh_l")
    FalloffRegionScaleFactors=(PositiveWidth=1,NegativeWidth=1,PositiveHeight=1,NegativeHeight=1)
    FalloffSize=0.25
    FlipHeightScaling=False
    FlipWidthScaling=False
    OptionalParentItem=(Type=Bone,Name="None")
    RotationOffset=(X=-60,Y=0,Z=-4)
   )
   n5 = &RigVMFunction_MathDoubleMul(A=@n4.OutputParam, B=RigVMVariableNode() {variable=Crouching_moving}.Value)
   > &RigUnit_ModifyTransforms(
    ItemToModify=((Item=(Type=Bone,Name="dyn_suspenders_01_f_MSTR"),Transform=(Rotation=(X=0.2682,Y=0.0947,Z=0.2856,W=0.9152),Translation=(X=-3.7701,Y=-1.5599,Z=1.1229),Scale3D=(X=1,Y=1,Z=1))))
    WeightMinimum=0
    Weight=@n5.Result
   )
   > &RigUnit_ModifyTransforms(
    ItemToModify=((Item=(Type=Bone,Name="dyn_belt_c_01"),Transform=(Rotation=(X=0,Y=0,Z=0,W=1),Translation=(X=2,Y=0,Z=0),Scale3D=(X=1,Y=1,Z=1))))
    WeightMinimum=0
    Weight=@n5.Result
   )
   > &RigUnit_ModifyTransforms(
    ItemToModify=((Item=(Type=Bone,Name="dyn_suspenders_02_s"),Transform=(Rotation=(X=0.2549,Y=0.1677,Z=0.0449,W=0.9513),Translation=(X=-13.3343,Y=3.1744,Z=-2.868),Scale3D=(X=1,Y=1,Z=1))))
    WeightMinimum=0
    Weight=@n5.Result
   )
   > &RigUnit_ModifyTransforms(
    ItemToModify=((Item=(Type=Bone,Name="dyn_suspenders_02_s2"),Transform=(Rotation=(X=0.0436,Y=0,Z=0,W=0.999),Translation=(X=-20.9148,Y=-0.3931,Z=-3.1463),Scale3D=(X=1,Y=1,Z=1))))
    WeightMinimum=0
    Weight=@n5.Result
   )
   > &RigUnit_ModifyTransforms(
    ItemToModify=((Item=(Type=Bone,Name="dyn_suspenders_02_f"),Transform=(Rotation=(X=0,Y=-0.3827,Z=0,W=0.9239),Translation=(X=0,Y=0,Z=0),Scale3D=(X=1,Y=1,Z=1))))
    WeightMinimum=0
    Weight=@n5.Result
   )
   > &RigUnit_ModifyTransforms(
    ItemToModify=((Item=(Type=Bone,Name="dyn_suspenders_01_b"),Transform=(Rotation=(X=-0.0933,Y=-0.2132,Z=0.1465,W=0.9615),Translation=(X=0,Y=0,Z=0),Scale3D=(X=1,Y=1,Z=1))))
    WeightMinimum=0
    Weight=@n5.Result
   )
   > &RigUnit_ModifyTransforms(
    ItemToModify=((Item=(Type=Bone,Name="dyn_suspenders_02_b"),Transform=(Rotation=(X=-0.0933,Y=-0.2132,Z=0.1465,W=0.9615),Translation=(X=-2.0496,Y=0.5484,Z=4.1804),Scale3D=(X=1,Y=1,Z=1))))
    WeightMinimum=0
    Weight=@n5.Result
   )
   > &RigUnit_ModifyTransforms(
    ItemToModify=((Item=(Type=Bone,Name="dyn_belt_l_02"),Transform=(Rotation=(X=0,Y=0,Z=0,W=1),Translation=(X=0,Y=1,Z=0),Scale3D=(X=1,Y=1,Z=1))))
    WeightMinimum=0
    Weight=@n5.Result
   )
   > n6 = &RigUnit_SphericalPoseReader(
    ActiveRegionScaleFactors=(PositiveWidth=1,NegativeWidth=3,PositiveHeight=2,NegativeHeight=1)
    ActiveRegionSize=0.07
    Debug=(bDrawDebug=false,bDraw2D=False,bDrawLocalAxes=False,DebugScale=25,DebugSegments=20,DebugThickness=0.25)
    DriverAxis=(X=1,Y=0,Z=0)
    DriverItem=(Type=Bone,Name="thigh_l")
    FalloffRegionScaleFactors=(PositiveWidth=1,NegativeWidth=2,PositiveHeight=2,NegativeHeight=1)
    FalloffSize=0.29
    FlipHeightScaling=False
    FlipWidthScaling=False
    OptionalParentItem=(Type=Bone,Name="None")
    RotationOffset=(X=-100,Y=0,Z=70)
   )
   > &RigUnit_ModifyTransforms(
    ItemToModify=((Item=(Type=Bone,Name="dyn_suspenders_01_f_MSTR"),Transform=(Rotation=(X=-0.017,Y=0.1294,Z=0.1294,W=0.983),Translation=(X=-2.5,Y=1.8,Z=0.0495),Scale3D=(X=1,Y=1,Z=1))))
    WeightMinimum=0
    Weight=&RigVMFunction_MathDoubleMul(A=@n6.OutputParam, B=RigVMVariableNode() {variable=Crouching_moving}.Value).Result
   )
   > n7 = &RigUnit_SphericalPoseReader(
    ActiveRegionScaleFactors=(PositiveWidth=1,NegativeWidth=1,PositiveHeight=1,NegativeHeight=1)
    ActiveRegionSize=0.05
    Debug=(bDrawDebug=false,bDraw2D=False,bDrawLocalAxes=False,DebugScale=25,DebugSegments=20,DebugThickness=0.25)
    DriverAxis=(X=1,Y=0,Z=0)
    DriverItem=(Type=Bone,Name="thigh_r")
    FalloffRegionScaleFactors=(PositiveWidth=1,NegativeWidth=1,PositiveHeight=1,NegativeHeight=1)
    FalloffSize=0.25
    FlipHeightScaling=False
    FlipWidthScaling=false
    OptionalParentItem=(Type=Bone,Name="None")
    RotationOffset=(X=-73,Y=0,Z=65)
   )
   > &RigUnit_ModifyTransforms(
    ItemToModify=((Item=(Type=Bone,Name="dyn_belt_r_01_MSTR"),Transform=(Rotation=(X=0,Y=0,Z=0,W=1),Translation=(X=-1,Y=0,Z=0),Scale3D=(X=1,Y=1,Z=1))))
    WeightMinimum=0
    Weight=@n7.OutputParam
   )
   > &RigUnit_ModifyTransforms(
    ItemToModify=((Item=(Type=Bone,Name="dyn_belt_ring_01"),Transform=(Rotation=(X=0.0038,Y=0.0871,Z=-0.0435,W=0.9952),Translation=(X=-3,Y=-1,Z=0),Scale3D=(X=1,Y=1,Z=1))))
    WeightMinimum=0
    Weight=@n7.OutputParam
   )
   n8 = &RigUnit_SphericalPoseReader(
    ActiveRegionScaleFactors=(PositiveWidth=2,NegativeWidth=1,PositiveHeight=1,NegativeHeight=1)
    ActiveRegionSize=0.05
    Debug=(bDrawDebug=false,bDraw2D=False,bDrawLocalAxes=False,DebugScale=25,DebugSegments=20,DebugThickness=0.25)
    DriverAxis=(X=1,Y=0,Z=0)
    DriverItem=(Type=Bone,Name="thigh_r")
    FalloffRegionScaleFactors=(PositiveWidth=1.5,NegativeWidth=1,PositiveHeight=1,NegativeHeight=1)
    FalloffSize=0.25
    FlipHeightScaling=False
    FlipWidthScaling=False
    OptionalParentItem=(Type=Bone,Name="None")
    RotationOffset=(X=-90,Y=0,Z=30)
    ExecutePin=@RigVMFunction_Sequence.C
   )
   n9 = &RigVMFunction_MathDoubleMul(A=@n8.OutputParam, B=RigVMVariableNode() {variable=Sprint}.Value)
   > &RigUnit_ModifyTransforms(
    ItemToModify=((Item=(Type=Bone,Name="dyn_belt_ring_01"),Transform=(Rotation=(X=0,Y=0.2164,Z=0,W=0.9763),Translation=(X=0,Y=2,Z=0),Scale3D=(X=1,Y=1,Z=1))))
    WeightMinimum=0
    Weight=@n9.Result
   )
   > &RigUnit_ModifyTransforms(
    ItemToModify=((Item=(Type=Bone,Name="dyn_belt_c_01"),Transform=(Rotation=(X=0,Y=0,Z=0,W=1),Translation=(X=-2.5,Y=1,Z=0),Scale3D=(X=1,Y=1,Z=1))))
    WeightMinimum=0
    Weight=@n9.Result
   )
   > &RigUnit_ModifyTransforms(
    ItemToModify=((Item=(Type=Bone,Name="dyn_belt_l_01"),Transform=(Rotation=(X=0,Y=0,Z=0,W=1),Translation=(X=-1,Y=1,Z=0),Scale3D=(X=1,Y=1,Z=1))))
    WeightMinimum=0
    Weight=@n9.Result
   )
   > &RigUnit_ModifyTransforms(
    ItemToModify=((Item=(Type=Bone,Name="dyn_belt_l_02"),Transform=(Rotation=(X=0,Y=0.342,Z=0,W=0.9397),Translation=(X=0,Y=1,Z=0),Scale3D=(X=1,Y=1,Z=1))))
    WeightMinimum=0
    Weight=@n9.Result
   )
   > &RigUnit_ModifyTransforms(
    ItemToModify=((Item=(Type=Bone,Name="dyn_suspenders_01_f_MSTR"),Transform=(Rotation=(X=0.0151,Y=0.173,Z=0.0858,W=0.9811),Translation=(X=-2.3167,Y=-0.202,Z=0.8432),Scale3D=(X=1,Y=1,Z=1))))
    WeightMinimum=0
    Weight=@n9.Result
   )
   > n10 = &RigUnit_SphericalPoseReader(
    ActiveRegionScaleFactors=(PositiveWidth=1,NegativeWidth=1,PositiveHeight=1,NegativeHeight=1)
    ActiveRegionSize=0.03
    Debug=(bDrawDebug=false,bDraw2D=False,bDrawLocalAxes=False,DebugScale=25,DebugSegments=20,DebugThickness=0.25)
    DriverAxis=(X=1,Y=0,Z=0)
    DriverItem=(Type=Bone,Name="thigh_r")
    FalloffRegionScaleFactors=(PositiveWidth=1,NegativeWidth=1,PositiveHeight=1,NegativeHeight=1)
    FalloffSize=0.2
    FlipHeightScaling=False
    FlipWidthScaling=false
    OptionalParentItem=(Type=Bone,Name="None")
    RotationOffset=(X=-105,Y=0,Z=85)
   )
   > &RigUnit_ModifyTransforms(
    ItemToModify=((Item=(Type=Bone,Name="dyn_suspenders_01_f_MSTR"),Transform=(Rotation=(X=0.017,Y=-0.1294,Z=0.1294,W=0.983),Translation=(X=0,Y=0,Z=0),Scale3D=(X=1,Y=1,Z=1))))
    WeightMinimum=0
    Weight=@n10.OutputParam
   )
   > n11 = &RigUnit_SphericalPoseReader(
    ActiveRegionScaleFactors=(PositiveWidth=1,NegativeWidth=1,PositiveHeight=1,NegativeHeight=1)
    ActiveRegionSize=0.05
    Debug=(bDrawDebug=false,bDraw2D=False,bDrawLocalAxes=False,DebugScale=25,DebugSegments=20,DebugThickness=0.25)
    DriverAxis=(X=1,Y=0,Z=0)
    DriverItem=(Type=Bone,Name="thigh_l")
    FalloffRegionScaleFactors=(PositiveWidth=1,NegativeWidth=2,PositiveHeight=2,NegativeHeight=1)
    FalloffSize=0.25
    FlipHeightScaling=False
    FlipWidthScaling=False
    OptionalParentItem=(Type=Bone,Name="None")
    RotationOffset=(X=-97,Y=0,Z=60)
   )
   > &RigUnit_ModifyTransforms(
    ItemToModify=((Item=(Type=Bone,Name="dyn_suspenders_01_f_MSTR"),Transform=(Rotation=(X=0,Y=0,Z=0,W=1),Translation=(X=0,Y=1,Z=0),Scale3D=(X=1,Y=1,Z=1))))
    WeightMinimum=0
    Weight=&RigVMFunction_MathDoubleMul(A=@n11.OutputParam, B=RigVMVariableNode() {variable=Sprint}.Value).Result
   )
   > n12 = &RigUnit_SphericalPoseReader(
    ActiveRegionScaleFactors=(PositiveWidth=1,NegativeWidth=1,PositiveHeight=1,NegativeHeight=1)
    ActiveRegionSize=0.08
    Debug=(bDrawDebug=false,bDraw2D=False,bDrawLocalAxes=False,DebugScale=25,DebugSegments=20,DebugThickness=0.25)
    DriverAxis=(X=1,Y=0,Z=0)
    DriverItem=(Type=Bone,Name="thigh_l")
    FalloffRegionScaleFactors=(PositiveWidth=1,NegativeWidth=1,PositiveHeight=1,NegativeHeight=1)
    FalloffSize=0.25
    FlipHeightScaling=False
    FlipWidthScaling=False
    OptionalParentItem=(Type=Bone,Name="None")
    RotationOffset=(X=-90,Y=0,Z=105)
   )
   > &RigUnit_ModifyTransforms(
    ItemToModify=((Item=(Type=Bone,Name="dyn_suspenders_01_f_MSTR"),Transform=(Rotation=(X=0,Y=0,Z=0.2164,W=0.9763),Translation=(X=0,Y=3,Z=0),Scale3D=(X=1,Y=1,Z=1))))
    WeightMinimum=0
    Weight=&RigVMFunction_MathDoubleMul(A=@n12.OutputParam, B=RigVMVariableNode() {variable=Sprint}.Value).Result
   )
   > n13 = &RigUnit_SphericalPoseReader(
    ActiveRegionScaleFactors=(PositiveWidth=1,NegativeWidth=1,PositiveHeight=1,NegativeHeight=1)
    ActiveRegionSize=0.05
    Debug=(bDrawDebug=false,bDraw2D=False,bDrawLocalAxes=False,DebugScale=25,DebugSegments=20,DebugThickness=0.25)
    DriverAxis=(X=1,Y=0,Z=0)
    DriverItem=(Type=Bone,Name="thigh_l")
    FalloffRegionScaleFactors=(PositiveWidth=1,NegativeWidth=1,PositiveHeight=1,NegativeHeight=1)
    FalloffSize=0.25
    FlipHeightScaling=False
    FlipWidthScaling=False
    OptionalParentItem=(Type=Bone,Name="None")
    RotationOffset=(X=-83,Y=0,Z=14)
   )
   > &RigUnit_ModifyTransforms(
    ItemToModify=((Item=(Type=Bone,Name="dyn_suspenders_01_f_MSTR"),Transform=(Rotation=(X=0,Y=0,Z=0.0872,W=0.9962),Translation=(X=0,Y=0,Z=0),Scale3D=(X=1,Y=1,Z=1))))
    WeightMinimum=0
    Weight=&RigVMFunction_MathDoubleMul(A=@n13.OutputParam, B=RigVMVariableNode() {variable=Sprint}.Value).Result
   )
   n14 = &RigUnit_SphericalPoseReader(
    ActiveRegionScaleFactors=(PositiveWidth=1,NegativeWidth=2,PositiveHeight=1,NegativeHeight=3)
    ActiveRegionSize=0.05
    Debug=(bDrawDebug=false,bDraw2D=False,bDrawLocalAxes=False,DebugScale=25,DebugSegments=20,DebugThickness=0.25)
    DriverAxis=(X=1,Y=0,Z=0)
    DriverItem=(Type=Bone,Name="upperarm_r")
    FalloffRegionScaleFactors=(PositiveWidth=1,NegativeWidth=1,PositiveHeight=1,NegativeHeight=2)
    FalloffSize=0.25
    FlipHeightScaling=False
    FlipWidthScaling=False
    OptionalParentItem=(Type=Bone,Name="None")
    RotationOffset=(X=95,Y=15,Z=4)
    ExecutePin=@RigVMFunction_Sequence.D
   )
   > &RigUnit_ModifyTransforms(
    ItemToModify=((Item=(Type=Bone,Name="dyn_suspenders_r_01_MSTR"),Transform=(Rotation=(X=-0.1736,Y=0,Z=0,W=0.9848),Translation=(X=0,Y=0,Z=2),Scale3D=(X=1,Y=1,Z=1))))
    WeightMinimum=0
    Weight=@n14.OutputParam
   )
   > n15 = &RigUnit_SphericalPoseReader(
    ActiveRegionScaleFactors=(PositiveWidth=1,NegativeWidth=1,PositiveHeight=1,NegativeHeight=1)
    ActiveRegionSize=0.05
    Debug=(bDrawDebug=false,bDraw2D=False,bDrawLocalAxes=False,DebugScale=25,DebugSegments=20,DebugThickness=0.25)
    DriverAxis=(X=1,Y=0,Z=0)
    DriverItem=(Type=Bone,Name="thigh_l")
    FalloffRegionScaleFactors=(PositiveWidth=1,NegativeWidth=1,PositiveHeight=1,NegativeHeight=1)
    FalloffSize=0.25
    FlipHeightScaling=False
    FlipWidthScaling=False
    OptionalParentItem=(Type=Bone,Name="None")
    RotationOffset=(X=-80,Y=0,Z=45)
   )
   > &RigUnit_ModifyTransforms(
    ItemToModify=((Item=(Type=Bone,Name="dyn_suspenders_01_f_MSTR"),Transform=(Rotation=(X=0,Y=0,Z=0,W=1),Translation=(X=-2,Y=0,Z=0),Scale3D=(X=1,Y=1,Z=1))))
    WeightMinimum=0
    Weight=@n15.OutputParam
   )
   n16 = &RigUnit_SphericalPoseReader(
    ActiveRegionScaleFactors=(PositiveWidth=1,NegativeWidth=1,PositiveHeight=1,NegativeHeight=1)
    ActiveRegionSize=0.05
    Debug=(bDrawDebug=false,bDraw2D=False,bDrawLocalAxes=False,DebugScale=25,DebugSegments=20,DebugThickness=0.25)
    DriverAxis=(X=-1,Y=0,Z=0)
    DriverItem=(Type=Bone,Name="thigh_l")
    FalloffRegionScaleFactors=(PositiveWidth=1,NegativeWidth=1.5,PositiveHeight=1,NegativeHeight=1)
    FalloffSize=0.2
    FlipHeightScaling=False
    FlipWidthScaling=False
    OptionalParentItem=(Type=Bone,Name="None")
    RotationOffset=(X=165,Y=10,Z=0)
    ExecutePin=@RigVMFunction_Sequence.E
   )
   > &RigUnit_ModifyTransforms(
    ItemToModify=((Item=(Type=Bone,Name="dyn_belt_l_02"),Transform=(Rotation=(X=0,Y=0.342,Z=0,W=0.9397),Translation=(X=0,Y=0,Z=1),Scale3D=(X=1,Y=1,Z=1))))
    WeightMinimum=0
    Weight=&RigVMFunction_MathDoubleMul(A=@n16.OutputParam, B=RigVMVariableNode() {variable=Emote}.Value).Result
   )
   > n17 = &RigUnit_SphericalPoseReader(
    ActiveRegionScaleFactors=(PositiveWidth=1,NegativeWidth=1,PositiveHeight=1,NegativeHeight=1)
    ActiveRegionSize=0.05
    Debug=(bDrawDebug=false,bDraw2D=False,bDrawLocalAxes=False,DebugScale=25,DebugSegments=20,DebugThickness=0.25)
    DriverAxis=(X=-1,Y=0,Z=0)
    DriverItem=(Type=Bone,Name="thigh_l")
    FalloffRegionScaleFactors=(PositiveWidth=1,NegativeWidth=1.5,PositiveHeight=1,NegativeHeight=1)
    FalloffSize=0.2
    FlipHeightScaling=False
    FlipWidthScaling=False
    OptionalParentItem=(Type=Bone,Name="None")
    RotationOffset=(X=145,Y=-35,Z=0)
   )
   > &RigUnit_ModifyTransforms(
    ItemToModify=((Item=(Type=Bone,Name="dyn_belt_l_02"),Transform=(Rotation=(X=0,Y=0.342,Z=0,W=0.9397),Translation=(X=0,Y=0,Z=1),Scale3D=(X=1,Y=1,Z=1))))
    WeightMinimum=0
    Weight=&RigVMFunction_MathDoubleMul(A=@n17.OutputParam, B=RigVMVariableNode() {variable=Emote}.Value).Result
   )
   > n18 = &RigUnit_SphericalPoseReader(
    ActiveRegionScaleFactors=(PositiveWidth=1,NegativeWidth=1,PositiveHeight=1,NegativeHeight=1)
    ActiveRegionSize=0.05
    Debug=(bDrawDebug=false,bDraw2D=False,bDrawLocalAxes=False,DebugScale=25,DebugSegments=20,DebugThickness=0.25)
    DriverAxis=(X=-1,Y=0,Z=0)
    DriverItem=(Type=Bone,Name="thigh_r")
    FalloffRegionScaleFactors=(PositiveWidth=1,NegativeWidth=1,PositiveHeight=1,NegativeHeight=1)
    FalloffSize=0.2
    FlipHeightScaling=False
    FlipWidthScaling=False
    OptionalParentItem=(Type=Bone,Name="None")
    RotationOffset=(X=95,Y=0,Z=29)
   )
   n19 = &RigVMFunction_MathDoubleMul(A=@n18.OutputParam, B=RigVMVariableNode() {variable=Emote}.Value)
   > &RigUnit_ModifyTransforms(
    ItemToModify=((Item=(Type=Bone,Name="dyn_belt_r_01_MSTR"),Transform=(Rotation=(X=0,Y=0,Z=0,W=1),Translation=(X=-1,Y=0,Z=0),Scale3D=(X=1,Y=1,Z=1))))
    WeightMinimum=0
    Weight=@n19.Result
   )
   > &RigUnit_ModifyTransforms(
    ItemToModify=((Item=(Type=Bone,Name="dyn_belt_ring_01_MSTR"),Transform=(Rotation=(X=0,Y=0.0872,Z=0,W=0.9962),Translation=(X=0,Y=0,Z=-2),Scale3D=(X=1,Y=1,Z=1))))
    WeightMinimum=0
    Weight=@n19.Result
   )
   > &RigUnit_ModifyTransforms(
    ItemToModify=((Item=(Type=Bone,Name="dyn_belt_c_01"),Transform=(Rotation=(X=0,Y=0,Z=0,W=1),Translation=(X=-2,Y=5,Z=0),Scale3D=(X=1,Y=1,Z=1))))
    WeightMinimum=0
    Weight=@n19.Result
   )
   > &RigUnit_ModifyTransforms(
    ItemToModify=((Item=(Type=Bone,Name="dyn_belt_l_01"),Transform=(Rotation=(X=0,Y=0.1305,Z=0,W=0.9914),Translation=(X=-2,Y=1,Z=0),Scale3D=(X=1,Y=1,Z=1))))
    WeightMinimum=0
    Weight=@n19.Result
   )
   > n20 = &RigUnit_SphericalPoseReader(
    ActiveRegionScaleFactors=(PositiveWidth=1,NegativeWidth=1,PositiveHeight=1,NegativeHeight=1)
    ActiveRegionSize=0.05
    Debug=(bDrawDebug=false,bDraw2D=False,bDrawLocalAxes=False,DebugScale=25,DebugSegments=20,DebugThickness=0.25)
    DriverAxis=(X=-1,Y=0,Z=0)
    DriverItem=(Type=Bone,Name="upperarm_r")
    FalloffRegionScaleFactors=(PositiveWidth=1,NegativeWidth=1,PositiveHeight=1,NegativeHeight=1)
    FalloffSize=0.2
    FlipHeightScaling=False
    FlipWidthScaling=False
    OptionalParentItem=(Type=Bone,Name="None")
    RotationOffset=(X=-5,Y=135,Z=0)
   )
   > &RigUnit_ModifyTransforms(
    ItemToModify=((Item=(Type=Bone,Name="dyn_suspenders_r_01_top"),Transform=(Rotation=(X=0,Y=0.5,Z=0,W=0.866),Translation=(X=-1,Y=0,Z=0),Scale3D=(X=1,Y=1,Z=1))))
    WeightMinimum=0
    Weight=@n20.OutputParam
   )
   > &RigUnit_ModifyTransforms(
    ItemToModify=((Item=(Type=Bone,Name="dyn_suspenders_r_01_MSTR"),Transform=(Rotation=(X=0.0872,Y=0,Z=0,W=0.9962),Translation=(X=0,Y=0,Z=0),Scale3D=(X=1,Y=1,Z=1))))
    WeightMinimum=0
    Weight=@n20.OutputParam
   )
   > n21 = &RigUnit_SphericalPoseReader(
    ActiveRegionScaleFactors=(PositiveWidth=1,NegativeWidth=3,PositiveHeight=3,NegativeHeight=1)
    ActiveRegionSize=0.06
    Debug=(bDrawDebug=false,bDraw2D=False,bDrawLocalAxes=False,DebugScale=25,DebugSegments=20,DebugThickness=0.25)
    DriverAxis=(X=-1,Y=0,Z=0)
    DriverItem=(Type=Bone,Name="thigh_r")
    FalloffRegionScaleFactors=(PositiveWidth=1,NegativeWidth=1,PositiveHeight=1,NegativeHeight=1)
    FalloffSize=0.25
    FlipHeightScaling=False
    FlipWidthScaling=False
    OptionalParentItem=(Type=Bone,Name="None")
    RotationOffset=(X=95,Y=0,Z=29)
   )
   n22 = &RigVMFunction_MathDoubleMul(A=@n21.OutputParam, B=RigVMVariableNode() {variable=Emote}.Value)
   > &RigUnit_ModifyTransforms(
    ItemToModify=((Item=(Type=Bone,Name="dyn_belt_ring_01_MSTR"),Transform=(Rotation=(X=-0.043,Y=0.1735,Z=0.0076,W=0.9839),Translation=(X=1.218,Y=1.7323,Z=-2.4434),Scale3D=(X=1,Y=1,Z=1))))
    WeightMinimum=0
    Weight=@n22.Result
   )
   > &RigUnit_ModifyTransforms(
    ItemToModify=((Item=(Type=Bone,Name="dyn_belt_c_01"),Transform=(Rotation=(X=0,Y=0,Z=0,W=1),Translation=(X=-2,Y=0,Z=0),Scale3D=(X=1,Y=1,Z=1))))
    WeightMinimum=0
    Weight=@n22.Result
   )
   > &RigUnit_ModifyTransforms(
    ItemToModify=((Item=(Type=Bone,Name="dyn_belt_l_01_MSTR"),Transform=(Rotation=(X=0.0436,Y=0.0436,Z=0.0019,W=0.9981),Translation=(X=0,Y=0,Z=0),Scale3D=(X=1,Y=1,Z=1))))
    WeightMinimum=0
    Weight=@n22.Result
   )
   > n23 = &RigUnit_SphericalPoseReader(
    ActiveRegionScaleFactors=(PositiveWidth=1,NegativeWidth=5,PositiveHeight=4,NegativeHeight=1)
    ActiveRegionSize=0.05
    Debug=(bDrawDebug=false,bDraw2D=False,bDrawLocalAxes=False,DebugScale=25,DebugSegments=20,DebugThickness=0.25)
    DriverAxis=(X=1,Y=0,Z=0)
    DriverItem=(Type=Bone,Name="thigh_l")
    FalloffRegionScaleFactors=(PositiveWidth=1,NegativeWidth=4,PositiveHeight=1,NegativeHeight=1)
    FalloffSize=0.29
    FlipHeightScaling=False
    FlipWidthScaling=False
    OptionalParentItem=(Type=Bone,Name="None")
    RotationOffset=(X=-80,Y=0,Z=65)
   )
   > &RigUnit_ModifyTransforms(
    ItemToModify=((Item=(Type=Bone,Name="dyn_suspenders_01_f_MSTR"),Transform=(Rotation=(X=0,Y=0,Z=0.3007,W=0.9537),Translation=(X=0,Y=0,Z=0),Scale3D=(X=1,Y=1,Z=1))))
    WeightMinimum=0
    Weight=&RigVMFunction_MathDoubleMul(A=@n23.OutputParam, B=RigVMVariableNode() {variable=Emote}.Value).Result
   )
   > n24 = &RigUnit_SphericalPoseReader(
    ActiveRegionScaleFactors=(PositiveWidth=1,NegativeWidth=1,PositiveHeight=1,NegativeHeight=1)
    ActiveRegionSize=0.06
    Debug=(bDrawDebug=false,bDraw2D=False,bDrawLocalAxes=False,DebugScale=25,DebugSegments=20,DebugThickness=0.25)
    DriverAxis=(X=-1,Y=0,Z=0)
    DriverItem=(Type=Bone,Name="thigh_r")
    FalloffRegionScaleFactors=(PositiveWidth=1,NegativeWidth=1,PositiveHeight=1,NegativeHeight=1)
    FalloffSize=0.25
    FlipHeightScaling=False
    FlipWidthScaling=False
    OptionalParentItem=(Type=Bone,Name="None")
    RotationOffset=(X=95,Y=0,Z=0)
   )
   n25 = &RigVMFunction_MathDoubleMul(A=@n24.OutputParam, B=RigVMVariableNode() {variable=Emote}.Value)
   > &RigUnit_ModifyTransforms(
    ItemToModify=((Item=(Type=Bone,Name="dyn_belt_ring_01_MSTR"),Transform=(Rotation=(X=-0.043,Y=0.1735,Z=0.0076,W=0.9839),Translation=(X=1.218,Y=1.7323,Z=-2.4434),Scale3D=(X=1,Y=1,Z=1))))
    WeightMinimum=0
    Weight=@n25.Result
   )
   > &RigUnit_ModifyTransforms(
    ItemToModify=((Item=(Type=Bone,Name="dyn_belt_c_01"),Transform=(Rotation=(X=0,Y=0,Z=0.5373,W=0.8434),Translation=(X=-3.7364,Y=3.4013,Z=1.0414),Scale3D=(X=1,Y=1,Z=1))))
    WeightMinimum=0
    Weight=@n25.Result
   )
   > &RigUnit_ModifyTransforms(
    ItemToModify=((Item=(Type=Bone,Name="dyn_belt_l_01_MSTR"),Transform=(Rotation=(X=0,Y=0.2164,Z=0,W=0.9763),Translation=(X=0.7344,Y=0,Z=2.4402),Scale3D=(X=1,Y=1,Z=1))))
    WeightMinimum=0
    Weight=@n25.Result
   )
   > n26 = &RigUnit_SphericalPoseReader(
    ActiveRegionScaleFactors=(PositiveWidth=1,NegativeWidth=1,PositiveHeight=1,NegativeHeight=1)
    ActiveRegionSize=0.06
    Debug=(bDrawDebug=false,bDraw2D=False,bDrawLocalAxes=False,DebugScale=25,DebugSegments=20,DebugThickness=0.25)
    DriverAxis=(X=-1,Y=0,Z=0)
    DriverItem=(Type=Bone,Name="thigh_r")
    FalloffRegionScaleFactors=(PositiveWidth=1,NegativeWidth=1,PositiveHeight=1,NegativeHeight=1)
    FalloffSize=0.25
    FlipHeightScaling=False
    FlipWidthScaling=False
    OptionalParentItem=(Type=Bone,Name="None")
    RotationOffset=(X=135,Y=0,Z=59)
   )
   > &RigUnit_ModifyTransforms(
    ItemToModify=((Item=(Type=Bone,Name="dyn_belt_ring_01_MSTR"),Transform=(Rotation=(X=-0.043,Y=0.1735,Z=0.0076,W=0.9839),Translation=(X=1.218,Y=1.7323,Z=-2.4434),Scale3D=(X=1,Y=1,Z=1))))
    WeightMinimum=0
    Weight=&RigVMFunction_MathDoubleMul(A=@n26.OutputParam, B=RigVMVariableNode() {variable=Emote}.Value).Result
   )
   > n27 = &RigUnit_SphericalPoseReader(
    ActiveRegionScaleFactors=(PositiveWidth=1,NegativeWidth=1,PositiveHeight=1,NegativeHeight=1)
    ActiveRegionSize=0.09
    Debug=(bDrawDebug=true,bDraw2D=False,bDrawLocalAxes=False,DebugScale=25,DebugSegments=20,DebugThickness=0.25)
    DriverAxis=(X=1,Y=0,Z=0)
    DriverItem=(Type=Bone,Name="thigh_l")
    FalloffRegionScaleFactors=(PositiveWidth=0.5,NegativeWidth=1,PositiveHeight=1,NegativeHeight=1)
    FalloffSize=0.29
    FlipHeightScaling=False
    FlipWidthScaling=False
    OptionalParentItem=(Type=Bone,Name="None")
    RotationOffset=(X=-80,Y=0,Z=-10)
   )
   n28 = &RigVMFunction_MathDoubleMul(A=@n27.OutputParam, B=RigVMVariableNode() {variable=Emote}.Value)
   > &RigUnit_ModifyTransforms(
    ItemToModify=((Item=(Type=Bone,Name="dyn_suspenders_01_f_MSTR"),Transform=(Rotation=(X=0.0757,Y=0.181,Z=0.3161,W=0.9282),Translation=(X=-0.8968,Y=1.6705,Z=0.1582),Scale3D=(X=1,Y=1,Z=1))))
    WeightMinimum=0
    Weight=@n28.Result
   )
   > &RigUnit_ModifyTransforms(
    ItemToModify=((Item=(Type=Bone,Name="dyn_belt_l_01_MSTR"),Transform=(Rotation=(X=0,Y=-0.2164,Z=0,W=0.9763),Translation=(X=0,Y=1.4357,Z=-1.1144),Scale3D=(X=1,Y=1,Z=1))))
    WeightMinimum=0
    Weight=@n28.Result
   )
   > n29 = &RigUnit_SphericalPoseReader(
    ActiveRegionScaleFactors=(PositiveWidth=1,NegativeWidth=1,PositiveHeight=1,NegativeHeight=1)
    ActiveRegionSize=0.05
    Debug=(bDrawDebug=false,bDraw2D=False,bDrawLocalAxes=False,DebugScale=25,DebugSegments=20,DebugThickness=0.25)
    DriverAxis=(X=0,Y=1,Z=0)
    DriverItem=(Type=Bone,Name="thigh_l")
    FalloffRegionScaleFactors=(PositiveWidth=1,NegativeWidth=1,PositiveHeight=1,NegativeHeight=1)
    FalloffSize=0.15
    FlipHeightScaling=False
    FlipWidthScaling=False
    OptionalParentItem=(Type=Bone,Name="None")
    RotationOffset=(X=110,Y=0,Z=17)
   )
   > &RigUnit_ModifyTransforms(
    ItemToModify=((Item=(Type=Bone,Name="dyn_suspenders_01_f_MSTR"),Transform=(Rotation=(X=0,Y=0,Z=0.1305,W=0.9914),Translation=(X=0,Y=0,Z=0),Scale3D=(X=1,Y=1,Z=1))))
    WeightMinimum=0
    Weight=&RigVMFunction_MathDoubleMul(A=@n29.OutputParam, B=RigVMVariableNode() {variable=Emote}.Value).Result
   )
   n30 = &RigUnit_SphericalPoseReader(
    ActiveRegionScaleFactors=(PositiveWidth=1,NegativeWidth=1,PositiveHeight=1,NegativeHeight=1)
    ActiveRegionSize=0.01
    Debug=(bDrawDebug=false,bDraw2D=False,bDrawLocalAxes=False,DebugScale=25,DebugSegments=20,DebugThickness=0.25)
    DriverAxis=(X=0,Y=1,Z=0)
    DriverItem=(Type=Bone,Name="spine_05")
    FalloffRegionScaleFactors=(PositiveWidth=1,NegativeWidth=1,PositiveHeight=1,NegativeHeight=0.5)
    FalloffSize=0.1
    FlipHeightScaling=False
    FlipWidthScaling=False
    OptionalParentItem=(Type=Bone,Name="None")
    RotationOffset=(X=100,Y=0,Z=10)
    ExecutePin=@RigVMFunction_Sequence.F
   )
   Subtract = &RigVMFunction_MathDoubleSub(A=@n30.OutputParam, B=RigVMVariableNode() {variable=Crouching_moving}.Value)
   > &RigUnit_ModifyTransforms(
    ItemToModify=((Item=(Type=Bone,Name="dyn_belt_l_02"),Transform=(Rotation=(X=0.0872,Y=0,Z=0,W=0.9962),Translation=(X=0,Y=0,Z=0),Scale3D=(X=1,Y=1,Z=1))))
    WeightMinimum=0
    Weight=@Subtract.Result
   )
   > n31 = &RigUnit_SphericalPoseReader(
    ActiveRegionScaleFactors=(PositiveWidth=1,NegativeWidth=1,PositiveHeight=1,NegativeHeight=1)
    ActiveRegionSize=0.01
    Debug=(bDrawDebug=false,bDraw2D=False,bDrawLocalAxes=False,DebugScale=25,DebugSegments=20,DebugThickness=0.25)
    DriverAxis=(X=0,Y=1,Z=0)
    DriverItem=(Type=Bone,Name="spine_05")
    FalloffRegionScaleFactors=(PositiveWidth=0.7,NegativeWidth=0.7,PositiveHeight=0.7,NegativeHeight=0.7)
    FalloffSize=0.1
    FlipHeightScaling=False
    FlipWidthScaling=False
    OptionalParentItem=(Type=Bone,Name="None")
    RotationOffset=(X=87,Y=0,Z=-2)
   )
   n32 = RigVMVariableNode() {variable=Crouching_moving}
   > &RigUnit_ModifyTransforms(
    ItemToModify=((Item=(Type=Bone,Name="dyn_belt_l_02"),Transform=(Rotation=(X=0.0872,Y=0,Z=0,W=0.9962),Translation=(X=0,Y=0,Z=0),Scale3D=(X=1,Y=1,Z=1))))
    WeightMinimum=0
    Weight=&RigVMFunction_MathDoubleSub(A=@n31.OutputParam, B=@n32.Value).Result
   )
   > n33 = &RigUnit_SphericalPoseReader(
    ActiveRegionScaleFactors=(PositiveWidth=3,NegativeWidth=3,PositiveHeight=1,NegativeHeight=1)
    ActiveRegionSize=0.02
    Debug=(bDrawDebug=false,bDraw2D=False,bDrawLocalAxes=False,DebugScale=25,DebugSegments=20,DebugThickness=0.25)
    DriverAxis=(X=1,Y=0,Z=0)
    DriverItem=(Type=Bone,Name="thigh_r")
    FalloffRegionScaleFactors=(PositiveWidth=2,NegativeWidth=2,PositiveHeight=1,NegativeHeight=1)
    FalloffSize=0.15
    FlipHeightScaling=False
    FlipWidthScaling=False
    OptionalParentItem=(Type=Bone,Name="None")
    RotationOffset=(X=-20,Y=-80,Z=0)
   )
   n34 = &RigVMFunction_MathDoubleAdd(B=RigVMVariableNode() {variable=Parachuting}.Value, A=@n32.Value)
   > &RigUnit_ModifyTransforms(
    ItemToModify=((Item=(Type=Bone,Name="dyn_belt_ring_01_MSTR"),Transform=(Rotation=(X=0,Y=0.0872,Z=0,W=0.9962),Translation=(X=2.6998,Y=2.4239,Z=-0.9713),Scale3D=(X=1,Y=1,Z=1))))
    WeightMinimum=0
    Weight=&RigVMFunction_MathDoubleSub(A=@n33.OutputParam, B=@n34.Result).Result
   )
   > n35 = &RigUnit_SphericalPoseReader(
    ActiveRegionScaleFactors=(PositiveWidth=1,NegativeWidth=1,PositiveHeight=1,NegativeHeight=1)
    ActiveRegionSize=0.07
    Debug=(bDrawDebug=false,bDraw2D=False,bDrawLocalAxes=False,DebugScale=25,DebugSegments=20,DebugThickness=0.25)
    DriverAxis=(X=-1,Y=0,Z=0)
    DriverItem=(Type=Bone,Name="thigh_l")
    FalloffRegionScaleFactors=(PositiveWidth=1,NegativeWidth=1,PositiveHeight=1,NegativeHeight=1)
    FalloffSize=0.15
    FlipHeightScaling=False
    FlipWidthScaling=False
    OptionalParentItem=(Type=Bone,Name="None")
    RotationOffset=(X=20,Y=90,Z=0)
   )
   n36 = &RigVMFunction_MathDoubleSub(A=@n35.OutputParam, B=@n34.Result)
   > &RigUnit_ModifyTransforms(
    ItemToModify=((Item=(Type=Bone,Name="dyn_suspenders_01_f_MSTR"),Transform=(Rotation=(X=0.0149,Y=0.0885,Z=0.1704,W=0.9813),Translation=(X=0,Y=0,Z=0),Scale3D=(X=1,Y=1,Z=1))))
    WeightMinimum=0
    Weight=@n36.Result
   )
   > &RigUnit_ModifyTransforms(
    ItemToModify=((Item=(Type=Bone,Name="dyn_suspenders_01_s_MSTR"),Transform=(Rotation=(X=0.0136,Y=-0.0039,Z=-0.0454,W=0.9989),Translation=(X=1.8949,Y=4.8156,Z=0.5699),Scale3D=(X=1,Y=1,Z=1))))
    WeightMinimum=0
    Weight=@n36.Result
   )
   > &RigUnit_ModifyTransforms(
    ItemToModify=((Item=(Type=Bone,Name="dyn_suspenders_02_s"),Transform=(Rotation=(X=0,Y=0,Z=0,W=1),Translation=(X=0,Y=6.3966,Z=-2.3546),Scale3D=(X=1,Y=1,Z=1))))
    WeightMinimum=0
    Weight=@n36.Result
   )
   > n37 = &RigUnit_SphericalPoseReader(
    ActiveRegionScaleFactors=(PositiveWidth=1.5,NegativeWidth=1.8,PositiveHeight=1.8,NegativeHeight=1.5)
    ActiveRegionSize=0.06
    Debug=(bDrawDebug=false,bDraw2D=False,bDrawLocalAxes=False,DebugScale=25,DebugSegments=20,DebugThickness=0.25)
    DriverAxis=(X=-1,Y=0,Z=0)
    DriverItem=(Type=Bone,Name="thigh_l")
    FalloffRegionScaleFactors=(PositiveWidth=1,NegativeWidth=1,PositiveHeight=1,NegativeHeight=1)
    FalloffSize=0.15
    FlipHeightScaling=False
    FlipWidthScaling=False
    OptionalParentItem=(Type=Bone,Name="None")
    RotationOffset=(X=62,Y=112,Z=0)
   )
   n38 = &RigVMFunction_MathDoubleSub(A=@n37.OutputParam, B=@n34.Result)
   > &RigUnit_ModifyTransforms(
    ItemToModify=((Item=(Type=Bone,Name="dyn_suspenders_01_f_MSTR"),Transform=(Rotation=(X=0.1743,Y=-0.0136,Z=0.1456,W=0.9738),Translation=(X=0,Y=0,Z=0),Scale3D=(X=1,Y=1,Z=1))))
    WeightMinimum=0
    Weight=@n38.Result
   )
   > &RigUnit_ModifyTransforms(
    ItemToModify=((Item=(Type=Bone,Name="dyn_suspenders_01_s_MSTR"),Transform=(Rotation=(X=0.0454,Y=0.0451,Z=-0.042,W=0.9971),Translation=(X=0,Y=0,Z=0),Scale3D=(X=1,Y=1,Z=1))))
    WeightMinimum=0
    Weight=@n38.Result
   )
   > n39 = &RigUnit_SphericalPoseReader(
    ActiveRegionScaleFactors=(PositiveWidth=1,NegativeWidth=1,PositiveHeight=1,NegativeHeight=1)
    ActiveRegionSize=0.07
    Debug=(bDrawDebug=false,bDraw2D=False,bDrawLocalAxes=False,DebugScale=25,DebugSegments=20,DebugThickness=0.25)
    DriverAxis=(X=-1,Y=0,Z=0)
    DriverItem=(Type=Bone,Name="thigh_l")
    FalloffRegionScaleFactors=(PositiveWidth=1,NegativeWidth=1,PositiveHeight=1,NegativeHeight=1)
    FalloffSize=0.15
    FlipHeightScaling=False
    FlipWidthScaling=False
    OptionalParentItem=(Type=Bone,Name="None")
    RotationOffset=(X=10,Y=105,Z=0)
   )
   > &RigUnit_ModifyTransforms(
    ItemToModify=((Item=(Type=Bone,Name="dyn_suspenders_01_f_MSTR"),Transform=(Rotation=(X=0,Y=0,Z=0,W=1),Translation=(X=0,Y=0,Z=0),Scale3D=(X=1,Y=1,Z=1))))
    WeightMinimum=0
    Weight=&RigVMFunction_MathDoubleMul(A=@n39.OutputParam, B=RigVMVariableNode() {variable=Skydiving}.Value).Result
   )
   > n40 = &RigUnit_SphericalPoseReader(
    ActiveRegionScaleFactors=(PositiveWidth=1.5,NegativeWidth=1.6,PositiveHeight=1.5,NegativeHeight=1.5)
    ActiveRegionSize=0.05
    Debug=(bDrawDebug=false,bDraw2D=False,bDrawLocalAxes=False,DebugScale=25,DebugSegments=20,DebugThickness=0.25)
    DriverAxis=(X=-1,Y=0,Z=0)
    DriverItem=(Type=Bone,Name="thigh_l")
    FalloffRegionScaleFactors=(PositiveWidth=1,NegativeWidth=3.5,PositiveHeight=1,NegativeHeight=1)
    FalloffSize=0.13
    FlipHeightScaling=False
    FlipWidthScaling=False
    OptionalParentItem=(Type=Bone,Name="None")
    RotationOffset=(X=107,Y=0,Z=15)
   )
   n41 = RigVMVariableNode() {variable=Jumping}
   > &RigUnit_ModifyTransforms(
    ItemToModify=((Item=(Type=Bone,Name="dyn_suspenders_01_f_MSTR"),Transform=(Rotation=(X=0.0205,Y=-0.0766,Z=0.3411,W=0.9367),Translation=(X=0,Y=0,Z=0),Scale3D=(X=1,Y=1,Z=1))))
    WeightMinimum=0
    Weight=&RigVMFunction_MathDoubleMul(A=&RigVMFunction_MathDoubleSub(A=@n40.OutputParam, B=@n34.Result).Result, B=@n41.Value).Result
   )
   > n42 = &RigUnit_SphericalPoseReader(
    ActiveRegionScaleFactors=(PositiveWidth=1,NegativeWidth=1,PositiveHeight=1,NegativeHeight=1)
    ActiveRegionSize=0.05
    Debug=(bDrawDebug=false,bDraw2D=False,bDrawLocalAxes=False,DebugScale=25,DebugSegments=20,DebugThickness=0.25)
    DriverAxis=(X=-1,Y=0,Z=0)
    DriverItem=(Type=Bone,Name="thigh_l")
    FalloffRegionScaleFactors=(PositiveWidth=1,NegativeWidth=1,PositiveHeight=1,NegativeHeight=1)
    FalloffSize=0.13
    FlipHeightScaling=False
    FlipWidthScaling=False
    OptionalParentItem=(Type=Bone,Name="None")
    RotationOffset=(X=90,Y=0,Z=5)
   )
   n43 = RigVMVariableNode() {variable=Parachuting}
   > &RigUnit_ModifyTransforms(
    ItemToModify=((Item=(Type=Bone,Name="dyn_suspenders_01_f_MSTR"),Transform=(Rotation=(X=0.2272,Y=-0.0037,Z=0.3314,W=0.9157),Translation=(X=0,Y=0,Z=0),Scale3D=(X=1,Y=1,Z=1))))
    WeightMinimum=0
    Weight=&RigVMFunction_MathDoubleMul(A=@n42.OutputParam, B=@n43.Value).Result
   )
   > n44 = &RigUnit_SphericalPoseReader(
    ActiveRegionScaleFactors=(PositiveWidth=1,NegativeWidth=1,PositiveHeight=1,NegativeHeight=1)
    ActiveRegionSize=0.07
    Debug=(bDrawDebug=false,bDraw2D=False,bDrawLocalAxes=False,DebugScale=25,DebugSegments=20,DebugThickness=0.25)
    DriverAxis=(X=-1,Y=0,Z=0)
    DriverItem=(Type=Bone,Name="thigh_l")
    FalloffRegionScaleFactors=(PositiveWidth=1,NegativeWidth=1,PositiveHeight=1,NegativeHeight=1)
    FalloffSize=0.15
    FlipHeightScaling=False
    FlipWidthScaling=False
    OptionalParentItem=(Type=Bone,Name="None")
    RotationOffset=(X=-16,Y=80,Z=0)
   )
   > &RigUnit_ModifyTransforms(
    ItemToModify=((Item=(Type=Bone,Name="dyn_suspenders_01_f_MSTR"),Transform=(Rotation=(X=0,Y=0,Z=0.1305,W=0.9914),Translation=(X=0,Y=0,Z=0),Scale3D=(X=1,Y=1,Z=1))))
    WeightMinimum=0
    Weight=&RigVMFunction_MathDoubleMul(A=@n44.OutputParam, B=@n43.Value).Result
   )
   n45 = &RigUnit_SphericalPoseReader(
    ActiveRegionScaleFactors=(PositiveWidth=1,NegativeWidth=1,PositiveHeight=1,NegativeHeight=1)
    ActiveRegionSize=0.1
    Debug=(bDrawDebug=false,bDraw2D=False,bDrawLocalAxes=False,DebugScale=25,DebugSegments=20,DebugThickness=0.25)
    DriverAxis=(X=-1,Y=0,Z=0)
    DriverItem=(Type=Bone,Name="clavicle_r")
    FalloffRegionScaleFactors=(PositiveWidth=1,NegativeWidth=1,PositiveHeight=1,NegativeHeight=1)
    FalloffSize=0.2
    FlipHeightScaling=False
    FlipWidthScaling=False
    OptionalParentItem=(Type=Bone,Name="None")
    RotationOffset=(X=0,Y=140,Z=0)
    ExecutePin=@RigVMFunction_Sequence.G
   )
   > &RigUnit_ModifyTransforms(
    ItemToModify=((Item=(Type=Bone,Name="dyn_suspenders_r_01_top_MSTR"),Transform=(Rotation=(X=0,Y=-0.0872,Z=0,W=0.9962),Translation=(X=-2.1428,Y=-0.7298,Z=-3.1516),Scale3D=(X=1,Y=1,Z=1))))
    WeightMinimum=0
    Weight=@n45.OutputParam
   )
   > &RigUnit_ModifyTransforms(
    ItemToModify=((Item=(Type=Bone,Name="dyn_suspenders_r_01_MSTR"),Transform=(Rotation=(X=0,Y=0,Z=0,W=1),Translation=(X=0,Y=0.5,Z=0.5),Scale3D=(X=1,Y=1,Z=1))))
    WeightMinimum=0
    Weight=@n45.OutputParam
   )
   > n46 = &RigUnit_SphericalPoseReader(
    ActiveRegionScaleFactors=(PositiveWidth=1,NegativeWidth=1,PositiveHeight=1,NegativeHeight=1)
    ActiveRegionSize=0.07
    Debug=(bDrawDebug=false,bDraw2D=False,bDrawLocalAxes=False,DebugScale=25,DebugSegments=20,DebugThickness=0.25)
    DriverAxis=(X=-1,Y=0,Z=0)
    DriverItem=(Type=Bone,Name="thigh_l")
    FalloffRegionScaleFactors=(PositiveWidth=1,NegativeWidth=1,PositiveHeight=1,NegativeHeight=1)
    FalloffSize=0.15
    FlipHeightScaling=False
    FlipWidthScaling=False
    OptionalParentItem=(Type=Bone,Name="None")
    RotationOffset=(X=30,Y=135,Z=0)
   )
   > &RigUnit_ModifyTransforms(
    ItemToModify=((Item=(Type=Bone,Name="dyn_suspenders_01_f_MSTR"),Transform=(Rotation=(X=0.0283,Y=-0.1274,Z=0.2146,W=0.9679),Translation=(X=0,Y=0,Z=0),Scale3D=(X=1,Y=1,Z=1))))
    WeightMinimum=0
    Weight=@n46.OutputParam
   )
   > &RigUnit_ModifyTransforms(
    ItemToModify=((Item=(Type=Bone,Name="dyn_suspenders_01_s_MSTR"),Transform=(Rotation=(X=-0.0114,Y=-0.2566,Z=0.13,W=0.9577),Translation=(X=0,Y=0,Z=0),Scale3D=(X=1,Y=1,Z=1))))
    WeightMinimum=0
    Weight=@n46.OutputParam
   )
   > &RigUnit_ModifyTransforms(
    ItemToModify=((Item=(Type=Bone,Name="dyn_suspenders_02_s"),Transform=(Rotation=(X=0,Y=0,Z=0,W=1),Translation=(X=0,Y=-5.6287,Z=-11.4208),Scale3D=(X=1,Y=1,Z=1))))
    WeightMinimum=0
    Weight=@n46.OutputParam
   )
   n47 = &RigUnit_SphericalPoseReader(
    ActiveRegionScaleFactors=(PositiveWidth=1,NegativeWidth=1,PositiveHeight=1,NegativeHeight=1)
    ActiveRegionSize=0.05
    Debug=(bDrawDebug=false,bDraw2D=False,bDrawLocalAxes=False,DebugScale=25,DebugSegments=20,DebugThickness=0.25)
    DriverAxis=(X=-1,Y=0,Z=0)
    DriverItem=(Type=Bone,Name="thigh_l")
    FalloffRegionScaleFactors=(PositiveWidth=0.5,NegativeWidth=0.5,PositiveHeight=0.3,NegativeHeight=0.3)
    FalloffSize=0.2
    FlipHeightScaling=False
    FlipWidthScaling=False
    OptionalParentItem=(Type=Bone,Name="pelvis")
    RotationOffset=(X=125,Y=0,Z=0)
    ExecutePin=@RigVMFunction_Sequence.H
   )
   > &RigUnit_ModifyTransforms(
    ItemToModify="((Item=(Type=Bone,Name=\"dyn_suspenders_01_f_MSTR\"),Transform=(Rotation=(X=0.1659,Y=-0.3202,Z=0.1659,W=0.9179),Translation=(X=0,Y=2,Z=0),Scale3D=(X=1,Y=1,Z=1))),(Item=(Type=Bone,Name=\"dyn_suspenders_01_s_MSTR\"),Transform=(Rotation=(X=0,Y=-0.1305,Z=0,W=0.991 ...(truncated)"
    WeightMinimum=0
    Weight=&RigVMFunction_MathDoubleMul(A=@n47.OutputParam, B=RigVMVariableNode() {variable=Emote}.Value).Result
   )
   > n48 = &RigUnit_SphericalPoseReader(
    ActiveRegionScaleFactors=(PositiveWidth=0.5,NegativeWidth=0.5,PositiveHeight=0.3,NegativeHeight=0.3)
    ActiveRegionSize=0.15
    Debug=(bDrawDebug=true,bDraw2D=False,bDrawLocalAxes=False,DebugScale=25,DebugSegments=20,DebugThickness=0.25)
    DriverAxis=(X=-1,Y=0,Z=0)
    DriverItem=(Type=Bone,Name="thigh_l")
    FalloffRegionScaleFactors=(PositiveWidth=0.5,NegativeWidth=0.5,PositiveHeight=0.2,NegativeHeight=0.2)
    FalloffSize=0.4
    FlipHeightScaling=False
    FlipWidthScaling=False
    OptionalParentItem=(Type=Bone,Name="pelvis")
    RotationOffset=(X=95,Y=0,Z=45)
   )
   > &RigUnit_ModifyTransforms(
    ItemToModify="((Item=(Type=Bone,Name=\"dyn_suspenders_01_f_MSTR\"),Transform=(Rotation=(X=0,Y=0,Z=0,W=1),Translation=(X=0,Y=-5,Z=0),Scale3D=(X=1,Y=1,Z=1))),(Item=(Type=Bone,Name=\"dyn_suspenders_02_f\"),Transform=(Rotation=(X=0,Y=0,Z=0,W=1),Translation=(X=0,Y=-10,Z=0),Scale ...(truncated)"
    WeightMinimum=0
    Weight=&RigVMFunction_MathDoubleMul(A=@n48.OutputParam, B=RigVMVariableNode() {variable=Emote}.Value).Result
   )
   Add = &RigVMFunction_MathDoubleAdd(A=RigVMVariableNode() {variable=StandingStill}.Value, B=@n41.Value)
   n49 = &RigUnit_SphericalPoseReader(
    ActiveRegionScaleFactors=(PositiveWidth=1,NegativeWidth=1,PositiveHeight=1,NegativeHeight=1)
    ActiveRegionSize=0.1
    Debug=(bDrawDebug=true,bDraw2D=False,bDrawLocalAxes=False,DebugScale=25,DebugSegments=20,DebugThickness=0.25)
    DriverAxis=(X=0,Y=1,Z=0)
    DriverItem=(Type=Bone,Name="spine_04")
    ExecutePin=()
    FalloffRegionScaleFactors=(PositiveWidth=1,NegativeWidth=1,PositiveHeight=1,NegativeHeight=1)
    FalloffSize=0.2
    FlipHeightScaling=False
    FlipWidthScaling=False
    OptionalParentItem=(Type=Bone,Name="None")
    RotationOffset=(X=75,Y=0,Z=0)
   )
   &RigVMFunction_MathDoubleAdd(B=@n49.OutputParam, A=@n46.OutputParam)
   &RigUnit_ModifyTransforms(
    ExecutePin=()
    ItemToModify=((Item=(Type=Bone,Name="dyn_suspenders_02_f"),Transform=(Rotation=(X=0,Y=0,Z=0,W=1),Translation=(X=0,Y=0,Z=0),Scale3D=(X=1,Y=1,Z=1))))
    Weight=1
    WeightMinimum=0
   )
   Multiply = &RigVMFunction_MathDoubleMul(A=@n36.Result, B=RigVMVariableNode() {variable=StandingStill}.Value)
   &RigVMFunction_MathDoubleMul(A=@n10.OutputParam, B=RigVMVariableNode() {variable=Sprint}.Value)
   &RigVMFunction_MathDoubleMul(A=@n15.OutputParam, B=RigVMVariableNode() {variable=Emote}.Value)
   &RigUnit_SphericalPoseReader(
    ActiveRegionScaleFactors=(PositiveWidth=1,NegativeWidth=1,PositiveHeight=1,NegativeHeight=1)
    ActiveRegionSize=0.05
    Debug=(bDrawDebug=true,bDraw2D=False,bDrawLocalAxes=False,DebugScale=25,DebugSegments=20,DebugThickness=0.25)
    DriverAxis=(X=-1,Y=0,Z=0)
    DriverItem=(Type=Bone,Name="thigh_r")
    ExecutePin=()
    FalloffRegionScaleFactors=(PositiveWidth=1,NegativeWidth=1,PositiveHeight=1,NegativeHeight=1)
    FalloffSize=0.2
    FlipHeightScaling=False
    FlipWidthScaling=False
    OptionalParentItem=(Type=Bone,Name="None")
    RotationOffset=(X=95,Y=0,Z=42)
   )
  ]}
]
hierarchy [
 root {loc=<0 0 0>, rot=<0 0 0 1>}
 pelvis {parent=root, loc=<0 -0.5301 98.0571>, rot=<0.0127 0.707 -0.0127 -0.707>}
 spine_01 {parent=pelvis, loc=<6.1208 -0.9328 0>, rot=<0 0 0.034 -0.9994>}
 spine_02 {parent=spine_01, loc=<6.5997 -0.2095 0>, rot=<0 0 0.0238 -0.9997>}
 spine_03 {parent=spine_02, loc=<6.4914 -0.1237 0>, rot=<0 0 0.0149 -0.9999>}
 spine_04 {parent=spine_03, loc=<6.9598 0.0794 0>, rot=<0 0 -0.0132 -0.9999>}
 spine_05 {parent=spine_04, loc=<6.9987 0 0>, rot=<0 0 -0.0015 -1>}
 attach_cape {parent=spine_05, loc=<9.0181 -11.7144 0>, rot=<0.0616 -0.7044 -0.0615 -0.7044>}
 neck_01 {parent=spine_05, loc=<11.3138 0.1529 0>, rot=<0 0 -0.2003 -0.9797>}
 neck_02 {parent=neck_01, loc=<4.5627 0 0>, rot=<0 0 0.0504 -0.9987>}
 head {parent=neck_02, loc=<5.3914 0 0>, rot=<0 0 0.115 -0.9934>}
 attach_backpack {parent=spine_05, loc=<-0.1211 -13.4735 0>, rot=<-0.0092 -0.7071 0.0094 -0.707>}
 clavicle_l {parent=spine_05, loc=<5.4831 -0.1053 -3.2946>, rot=<0.0465 -0.7289 -0.009 -0.6829>}
 upperarm_l {parent=clavicle_l, loc=<9.8658 -1.2136 -0.4105>, rot=<0.0164 -0.3245 -0.0269 -0.9454>}
 lowerarm_l {parent=upperarm_l, loc=<26.465 -0.8345 0.259>, rot=<0.0036 0.0161 -0.2298 -0.9731>}
 hand_l {parent=lowerarm_l, loc=<22.7355 -0.3203 -0.6105>, rot=<-0.6784 -0.0223 0.0735 -0.7307>}
 thumb_01_l {parent=hand_l, loc=<1.3118 -0.3191 -2.459>, rot=<0.6546 -0.4042 0.0664 -0.6354>}
 thumb_02_l {parent=thumb_01_l, loc=<5.9336 -0.0448 0.0401>, rot=<-0.0037 0.0213 0.298 -0.9543>}
 thumb_03_l {parent=thumb_02_l, loc=<2.2223 0.3682 -0.0907>, rot=<0.0118 -0.013 0.0066 -0.9998>}
 index_metacarpal_l {parent=hand_l, loc=<1.6535 0.0183 -2.96>, rot=<0.0158 -0.0124 -0.0173 -0.9996>}
 index_01_l {parent=index_metacarpal_l, loc=<8.3008 0.0357 -0.0524>, rot=<-0.0044 0.0485 0.0924 -0.9945>}
 index_02_l {parent=index_01_l, loc=<3.3587 -0.3004 0.0452>, rot=<-0.002 0.0027 0.3225 -0.9466>}
 index_03_l {parent=index_02_l, loc=<2.5578 -0.123 -0.0252>, rot=<-0.002 0.0067 0.2558 -0.9667>}
 middle_metacarpal_l {parent=hand_l, loc=<2.1856 -0.1725 -0.5984>, rot=<0.006 0.0168 -0.0329 -0.9993>}
 middle_01_l {parent=middle_metacarpal_l, loc=<7.3275 -0.2447 0.0129>, rot=<-0.0208 0.0042 0.0858 -0.9961>}
 middle_02_l {parent=middle_01_l, loc=<4.0123 0.1401 0.1175>, rot=<-0.0218 0.0258 0.3512 -0.9357>}
 middle_03_l {parent=middle_02_l, loc=<2.6077 0.0534 0.1373>, rot=<0.0082 0.0188 0.1708 -0.9851>}
 ring_metacarpal_l {parent=hand_l, loc=<2.2074 -0.3625 1.3638>, rot=<-0.0284 0.0722 -0.0458 -0.9959>}
 ring_01_l {parent=ring_metacarpal_l, loc=<7.0295 -0.6197 -0.1446>, rot=<-0.0489 -0.0252 0.1892 -0.9804>}
 ring_02_l {parent=ring_01_l, loc=<3.6174 0.4647 -0.1027>, rot=<0 0 0.2806 -0.9598>}
 ring_03_l {parent=ring_02_l, loc=<2.4152 0.114 0.0199>, rot=<0 0 0.1189 -0.9929>}
 pinky_metacarpal_l {parent=hand_l, loc=<1.7616 -0.9753 3.3814>, rot=<-0.0397 0.0814 -0.069 -0.9935>}
 pinky_01_l {parent=pinky_metacarpal_l, loc=<7.1709 -0.4831 -0.1094>, rot=<-0.0654 -0.0035 0.2242 -0.9723>}
 pinky_02_l {parent=pinky_01_l, loc=<2.8274 0.0893 -0.1043>, rot=<0.0037 0.0011 0.2854 -0.9584>}
 pinky_03_l {parent=pinky_02_l, loc=<2.2046 0.0307 -0.0078>, rot=<-0.0002 0.0098 0.1445 -0.9895>}
 lowerarm_twist_02_l {parent=lowerarm_l, loc=<10.6707 0 0>, rot=<0 0 0 1>}
 lowerarm_twist_01_l {parent=lowerarm_l, loc=<18.316 0 0>, rot=<0 0 0 1>}
 deform_elbow_in_base_l {parent=lowerarm_l, loc=<0.1834 3.6062 -0.2572>, rot=<0 0 0 -1>}
 deform_elbow_in_l {parent=deform_elbow_in_base_l, loc=<0 0 0>, rot=<0 0 0 1>}
 deform_elbow_out_base_l {parent=lowerarm_l, loc=<0 -4 0>, rot=<0 0 0 1>}
 deform_elbow_out_l {parent=deform_elbow_out_base_l, loc=<0 0 0>, rot=<0 0 0 1>}
 upperarm_twist_01_l {parent=upperarm_l, loc=<7.2135 -0.1134 0>, rot=<0 0 0 1>}
 upperarm_twist_02_l {parent=upperarm_l, loc=<19.7998 -0.1134 0>, rot=<0 0 0 1>}
 clavicle_r {parent=spine_05, loc=<5.4831 -0.1053 3.2946>, rot=<0.7289 0.0465 -0.6829 0.009>}
 upperarm_r {parent=clavicle_r, loc=<-9.8658 1.2136 0.4105>, rot=<0.0164 -0.3245 -0.0269 -0.9454>}
 lowerarm_r {parent=upperarm_r, loc=<-26.465 0.8345 -0.259>, rot=<0.0036 0.0161 -0.2298 -0.9731>}
 hand_r {parent=lowerarm_r, loc=<-22.7355 0.3203 0.6106>, rot=<-0.6784 -0.0223 0.0735 -0.7307>}
 thumb_01_r {parent=hand_r, loc=<-1.3118 0.3191 2.459>, rot=<0.6546 -0.4042 0.0664 -0.6354>}
 thumb_02_r {parent=thumb_01_r, loc=<-5.9336 0.0448 -0.0401>, rot=<-0.0037 0.0213 0.298 -0.9543>}
 thumb_03_r {parent=thumb_02_r, loc=<-2.2223 -0.3682 0.0907>, rot=<0.0118 -0.013 0.0066 -0.9998>}
 index_metacarpal_r {parent=hand_r, loc=<-1.6535 -0.0183 2.96>, rot=<0.0158 -0.0124 -0.0173 -0.9996>}
 index_01_r {parent=index_metacarpal_r, loc=<-8.3008 -0.0357 0.0524>, rot=<-0.0044 0.0485 0.0924 -0.9945>}
 index_02_r {parent=index_01_r, loc=<-3.3587 0.3004 -0.0452>, rot=<-0.002 0.0027 0.3225 -0.9466>}
 index_03_r {parent=index_02_r, loc=<-2.5578 0.123 0.0252>, rot=<-0.002 0.0067 0.2558 -0.9667>}
 middle_metacarpal_r {parent=hand_r, loc=<-2.1856 0.1725 0.5984>, rot=<0.006 0.0168 -0.0329 -0.9993>}
 middle_01_r {parent=middle_metacarpal_r, loc=<-7.3275 0.2447 -0.0129>, rot=<-0.0208 0.0042 0.0858 -0.9961>}
 middle_02_r {parent=middle_01_r, loc=<-4.0123 -0.1401 -0.1175>, rot=<-0.0218 0.0258 0.3512 -0.9357>}
 middle_03_r {parent=middle_02_r, loc=<-2.6077 -0.0534 -0.1373>, rot=<0.0082 0.0188 0.1708 -0.9851>}
 ring_metacarpal_r {parent=hand_r, loc=<-2.2074 0.3625 -1.3638>, rot=<-0.0284 0.0722 -0.0458 -0.9959>}
 ring_01_r {parent=ring_metacarpal_r, loc=<-7.0295 0.6197 0.1446>, rot=<-0.0489 -0.0252 0.1892 -0.9804>}
 ring_02_r {parent=ring_01_r, loc=<-3.6174 -0.4647 0.1027>, rot=<0 0 0.2806 -0.9598>}
 ring_03_r {parent=ring_02_r, loc=<-2.4152 -0.114 -0.0199>, rot=<0 0 0.1189 -0.9929>}
 pinky_metacarpal_r {parent=hand_r, loc=<-1.7616 0.9753 -3.3814>, rot=<-0.0397 0.0814 -0.069 -0.9935>}
 pinky_01_r {parent=pinky_metacarpal_r, loc=<-7.1709 0.4831 0.1094>, rot=<-0.0654 -0.0035 0.2242 -0.9723>}
 pinky_02_r {parent=pinky_01_r, loc=<-2.8274 -0.0893 0.1043>, rot=<0.0037 0.0011 0.2854 -0.9584>}
 pinky_03_r {parent=pinky_02_r, loc=<-2.2046 -0.0307 0.0078>, rot=<-0.0002 0.0098 0.1445 -0.9895>}
 lowerarm_twist_02_r {parent=lowerarm_r, loc=<-10.6707 0 0>, rot=<0 0 0 -1>}
 lowerarm_twist_01_r {parent=lowerarm_r, loc=<-18.316 0 0>, rot=<0 0 0 -1>}
 deform_elbow_in_base_r {parent=lowerarm_r, loc=<-0.1834 -3.6062 0.2572>, rot=<0 0 0 -1>}
 deform_elbow_in_r {parent=deform_elbow_in_base_r, loc=<0 0 0>, rot=<0 0 0 1>}
 deform_elbow_out_base_r {parent=lowerarm_r, loc=<0 4 -0.0001>, rot=<0 0 0 -1>}
 deform_elbow_out_r {parent=deform_elbow_out_base_r, loc=<0 0 0>, rot=<0 0 0 1>}
 upperarm_twist_01_r {parent=upperarm_r, loc=<-7.2135 0.1134 0>, rot=<0 0 0 -1>}
 upperarm_twist_02_r {parent=upperarm_r, loc=<-19.7998 0.1134 0>, rot=<0 0 0 -1>}
 dyn_suspenders_r_01_MSTR {parent=clavicle_r, loc=<-5.3196 -6.1125 -0.1024>, rot=<0.7052 -0.3202 0.6153 -0.1473>}
 dyn_suspenders_r_01 {parent=dyn_suspenders_r_01_MSTR, loc=<0 0 0>, rot=<0 0 0 1>}
 dyn_suspenders_r_01_top_MSTR {parent=clavicle_r, loc=<-4.4488 0.9364 -4.6908>, rot=<0.7573 0.0201 0.6526 -0.0151>}
 dyn_suspenders_r_01_top {parent=dyn_suspenders_r_01_top_MSTR, loc=<0 0 0>, rot=<0 0 0 1>}
 thigh_l {parent=pelvis, loc=<-1.0855 -0.2109 -8.7053>, rot=<0.0009 0.0355 0.0155 -0.9993>}
 calf_l {parent=thigh_l, loc=<-42.5284 -0.0019 -0.7297>, rot=<0.0006 -0.0161 -0.0391 -0.9991>}
 foot_l {parent=calf_l, loc=<-46.7401 -2.5014 -0.592>, rot=<0.0427 -0.0238 0.0353 -0.9982>}
 ball_l {parent=foot_l, loc=<-3.5658 10.8066 -0.1544>, rot=<0.0001 -0.0001 -0.7063 -0.7079>}
 calf_twist_02_l {parent=calf_l, loc=<-19.1546 -1.7943 -0.2638>, rot=<0.0237 0.0095 -0.0035 -0.9997>}
 calf_twist_01_l {parent=calf_l, loc=<-35.8483 -1.9197 -0.5794>, rot=<0.0237 0.0095 -0.0035 -0.9997>}
 deform_knee_in_base_l {parent=calf_l, loc=<0.8031 -5.0231 0.2456>, rot=<0 0 0 -1>}
 deform_knee_in_l {parent=deform_knee_in_base_l, loc=<0 0 0>, rot=<0 0 0 1>}
 deform_knee_out_base_l {parent=calf_l, loc=<0 4 0>, rot=<0 0 0 -1>}
 deform_knee_out_l {parent=deform_knee_out_base_l, loc=<0 0 0>, rot=<0 0 0 1>}
 dyn_knee_l_01_MSTR {parent=calf_l, loc=<-4.5999 8.0149 -2.3209>, rot=<-0.0374 -0.9838 0.1735 0.0264>}
 dyn_knee_l_01 {parent=dyn_knee_l_01_MSTR, loc=<0 0 0>, rot=<0 0 0 1>}
 thigh_twist_01_l {parent=thigh_l, loc=<-27.952 0 0>, rot=<0.0013 0 0 -1>}
 deform_groin_base_l {parent=thigh_l, loc=<0.4801 5.6776 -0.2486>, rot=<0 0 0 -1>}
 deform_groin_l {parent=deform_groin_base_l, loc=<0 0 0>, rot=<0 0 0 1>}
 thigh_r {parent=pelvis, loc=<-1.0855 -0.2109 8.7053>, rot=<0.0355 -0.0009 0.9993 0.0155>}
 calf_r {parent=thigh_r, loc=<42.5284 0.0019 0.7297>, rot=<0.0006 -0.0161 -0.0391 -0.9991>}
 foot_r {parent=calf_r, loc=<46.7401 2.5014 0.592>, rot=<0.0427 -0.0238 0.0353 -0.9982>}
 ball_r {parent=foot_r, loc=<3.5658 -10.8066 0.1544>, rot=<0.0001 -0.0001 -0.7063 -0.7079>}
 calf_twist_02_r {parent=calf_r, loc=<19.1546 1.7943 0.2638>, rot=<0.0237 0.0095 -0.0035 -0.9997>}
 calf_twist_01_r {parent=calf_r, loc=<35.8483 1.9197 0.5794>, rot=<0.0237 0.0095 -0.0035 -0.9997>}
 deform_knee_in_base_r {parent=calf_r, loc=<-0.8031 5.0231 -0.2457>, rot=<0 0 0 -1>}
 deform_knee_in_r {parent=deform_knee_in_base_r, loc=<0 0 0>, rot=<0 0 0 1>}
 deform_knee_out_base_r {parent=calf_r, loc=<0 -4 0>, rot=<0 0 0 -1>}
 deform_knee_out_r {parent=deform_knee_out_base_r, loc=<0 0 0>, rot=<0 0 0 1>}
 dyn_knee_r_01_MSTR {parent=calf_r, loc=<4.4651 -8.0433 1.2633>, rot=<0.9923 -0.0388 -0.0241 0.1155>}
 dyn_knee_r_01 {parent=dyn_knee_r_01_MSTR, loc=<0 0 0>, rot=<0 0 0 1>}
 thigh_twist_01_r {parent=thigh_r, loc=<27.952 0 0>, rot=<0.0013 0 0 -1>}
 deform_groin_base_r {parent=thigh_r, loc=<-0.4801 -5.6776 0.2486>, rot=<0 0 0 -1>}
 deform_groin_r {parent=deform_groin_base_r, loc=<0 0 0>, rot=<0 0 0 1>}
 dyn_belt_r_01_MSTR {parent=pelvis, loc=<-5.7418 1.4938 17.2344>, rot=<-0.0138 -0.7683 -0.6398 -0.0115>}
 dyn_belt_r_strap_01 {parent=dyn_belt_r_01_MSTR, loc=<0 0 0>, rot=<-0.012 -0.3668 -0.0522 -0.9288>}
 dyn_belt_r_strap_02 {parent=dyn_belt_r_strap_01, loc=<5.6795 0 0>, rot=<-0.1376 0.3002 -0.1371 -0.9339>}
 dyn_belt_r_strap_03 {parent=dyn_belt_r_strap_02, loc=<10.4997 0 0>, rot=<0 -0.1363 0 -0.9907>}
 dyn_belt_ring_01_MSTR {parent=pelvis, loc=<-4.3889 8.5428 9.9869>, rot=<-0.0179 -0.9998 0 0>}
 dyn_belt_ring_01 {parent=dyn_belt_ring_01_MSTR, loc=<0 0 0>, rot=<0 0 0 1>}
 deform_glute_base_l {parent=pelvis, loc=<-7.3466 -9.0213 -9.1245>, rot=<0.0172 0.031 -0.4477 -0.8935>}
 deform_glute_l {parent=deform_glute_base_l, loc=<0 0 0>, rot=<0 0 0 1>}
 deform_glute_base_r {parent=pelvis, loc=<-7.3464 -9.0213 9.1245>, rot=<0.031 -0.0172 0.8935 -0.4477>}
 deform_glute_r {parent=deform_glute_base_r, loc=<0 0 0>, rot=<0 0 0 1>}
 dyn_suspenders_01_base_l {parent=pelvis, loc=<5.4788 7.9576 -5.5252>, rot=<0.0211 -0.8783 -0.3849 -0.2829>}
 dyn_suspenders_01_l {parent=dyn_suspenders_01_base_l, loc=<0 0 0>, rot=<0 0 0 1>}
 dyn_suspenders_02_l {parent=dyn_suspenders_01_l, loc=<7.1298 0 0>, rot=<0.0361 -0.121 0.0108 -0.9919>}
 dyn_suspenders_03_l {parent=dyn_suspenders_02_l, loc=<12.2287 0 0>, rot=<0.2272 -0.0566 0.1183 -0.965>}
 dyn_suspenders_04_l {parent=dyn_suspenders_03_l, loc=<12.0314 0 0>, rot=<0.1001 -0.4419 0.0469 -0.8902>}
 dyn_suspenders_05_l {parent=dyn_suspenders_04_l, loc=<9.4141 0 0>, rot=<-0.1027 -0.6189 0.0084 -0.7787>}
 dyn_suspenders_06_l {parent=dyn_suspenders_05_l, loc=<10.278 0 0>, rot=<-0.0951 -0.1179 0.1309 -0.9797>}
 dyn_suspenders_07_l {parent=dyn_suspenders_06_l, loc=<8.1749 0 0>, rot=<-0.0768 -0.0434 0.0717 -0.9935>}
 dyn_suspenders_08_l {parent=dyn_suspenders_07_l, loc=<8.7759 0 0>, rot=<-0.0269 -0.1887 0.1229 -0.9739>}
 dyn_suspenders_09_l {parent=dyn_suspenders_08_l, loc=<9.3879 0 0>, rot=<-0.0052 -0.1201 0.1299 -0.9842>}
 dyn_suspenders_10_l {parent=dyn_suspenders_09_l, loc=<5.629 0 0>, rot=<0 -0.0246 0.9997 0>}
 dyn_suspenders_03_skin_l {parent=dyn_suspenders_03_l, loc=<0 0 0>, rot=<0 0 0 1>}
 dyn_suspenders_02_skin_l {parent=dyn_suspenders_02_l, loc=<0 0 0>, rot=<0 0 0 1>}
 dyn_suspenders_01_rotator {parent=dyn_suspenders_01_l, loc=<0 0 0>, rot=<0.2731 -0.2274 -0.1792 -0.9174>}
 dyn_suspenders_01_skin_l {parent=dyn_suspenders_01_l, loc=<0 0 0>, rot=<0 0 0 1>}
 dyn_suspenders_10_l_base_l {parent=pelvis, loc=<8.7334 -8.1214 -5.3214>, rot=<0.1854 -0.8288 0.4616 -0.256>}
 dyn_belt_l_01_MSTR {parent=pelvis, loc=<-0.1927 8.9584 -5.6719>, rot=<-0.0179 -0.9998 0 0>}
 dyn_belt_l_01 {parent=dyn_belt_l_01_MSTR, loc=<0 0 0>, rot=<0 0 0 1>}
 dyn_belt_l_02 {parent=dyn_belt_l_01, loc=<0 0 0>, rot=<0 0 0 1>}
 dyn_belt_c_01_MSTR {parent=pelvis, loc=<-2.6157 9.0453 1.6464>, rot=<-0.0179 -0.9998 0 0>}
 dyn_belt_c_01 {parent=dyn_belt_c_01_MSTR, loc=<0 0 0>, rot=<0 0 0 1>}
 dyn_belt_l2_01_MSTR {parent=pelvis, loc=<2.8808 2.4745 -12.944>, rot=<-0.0148 -0.8282 0.5602 0.01>}
 dyn_belt_l2_01 {parent=dyn_belt_l2_01_MSTR, loc=<0 0 0>, rot=<0 0 0 1>}
 dyn_suspenders_01_f_MSTR {parent=pelvis, loc=<5.652 7.6118 -8.4582>, rot=<-0.1404 -0.9773 0.1571 -0.0226>}
 dyn_suspenders_01_f {parent=dyn_suspenders_01_f_MSTR, loc=<0 0 0>, rot=<0 0 0 1>}
 dyn_suspenders_02_f {parent=dyn_suspenders_01_f, loc=<11.1056 0 0>, rot=<0 0 0 1>}
 dyn_suspenders_01_b_MSTR {parent=pelvis, loc=<9.0263 -8.3302 -6.0131>, rot=<0.2312 -0.9579 -0.1655 -0.04>}
 dyn_suspenders_01_b {parent=dyn_suspenders_01_b_MSTR, loc=<0 0 0>, rot=<0 0 0 1>}
 dyn_suspenders_02_b {parent=dyn_suspenders_01_b, loc=<11.1056 0 0>, rot=<0 0 0 1>}
 dyn_suspenders_01_s_MSTR {parent=pelvis, loc=<5.6952 -0.4251 -14.0889>, rot=<-0.0153 -0.9916 0.0023 -0.1283>}
 dyn_suspenders_01_s {parent=dyn_suspenders_01_s_MSTR, loc=<0 0 0>, rot=<0 0 0 1>}
 dyn_suspenders_02_s {parent=dyn_suspenders_01_s, loc=<30.6548 -0.6145 0.8057>, rot=<0 0 0 1>}
 dyn_suspenders_01_s2_MSTR {parent=pelvis, loc=<4.6837 -5.3995 -12.1437>, rot=<-0.0153 -0.9886 0.0027 -0.1495>}
 dyn_suspenders_01_s2 {parent=dyn_suspenders_01_s2_MSTR, loc=<0 0 0>, rot=<0 0 0 1>}
 dyn_suspenders_02_s2 {parent=dyn_suspenders_01_s2, loc=<29.0684 0 0>, rot=<0 0 0 1>}
]
variables [
 Crouching_moving {type=double, default=0}
 Emote {type=double, default=0}
 Jumping {type=double, default=0}
 Parachuting {type=double, default=0}
 Skydiving {type=double, default=0}
 Sprint {type=double, default=0}
 StandingStill {type=double, default=0}
]
