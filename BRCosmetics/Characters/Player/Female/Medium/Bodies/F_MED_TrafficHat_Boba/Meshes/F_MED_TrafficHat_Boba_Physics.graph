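# Physics asset for the F_MED_TrafficHat_Boba female medium player body cosmetic, giving capsule collision to the lower body only.
# Pelvis carries three capsules, while each thigh and calf has a single capsule, sized and placed differently on left and right.
# Constraints join each calf to its thigh and the left thigh to the pelvis with free angular motion.
# No constraint links the right thigh to the pelvis.
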PhysicsAsset /BRCosmetics/Characters/Player/Female/Medium/Bodies/F_MED_TrafficHat_Boba/Meshes/F_MED_TrafficHat_Boba_Physics
bodies {
 calf_l [{
   shapeType=Capsule
   length=30.8586
   radius=10.7352
   shapeName=calf_l_capsule
   center=<-22.2227 -8.0177 1.7309>
   rotation=<rot 83.5899 174.867 167.5275>
  }]
 calf_r [{shapeType=Capsule, length=22.4247, radius=13.6558, center=<21.9319 1.3919 -0.9849>, rotation=<rot 0 -78.9761 -85.638>}]
 pelvis [
  {shapeType=Capsule, length=9.5, radius=8, center=<0 -1.2267 0>}
  {shapeType=Capsule, length=5, radius=12, shapeName=pelvis_capsule, center=<-16.0249 -18.9836 2.0138>}
  {shapeType=Capsule, length=21, radius=15, shapeName=pelvis_capsule, center=<-9.5254 25.8148 1.9152>}
 ]
 thigh_l [{shapeType=Capsule, length=37.5181, radius=10.5, shapeName=thigh_l_capsule, center=<-20 -2.9455 0>, rotation=<rot -90 -179.9998 179.9997>}]
 thigh_r [{shapeType=Capsule, length=30, radius=13, center=<19.3861 4.8388 -4.596>, rotation=<rot 0 88.3103 -89.363>}]
}
constraints [
 {bone1Name=calf_l, bone2Name=thigh_l, swing1LimitDegrees=45, swing2LimitDegrees=45, twistLimitDegrees=45}
 {bone1Name=calf_r, bone2Name=thigh_r, swing1LimitDegrees=45, swing2LimitDegrees=45, twistLimitDegrees=45}
 {
  bone1Name=thigh_l
  bone2Name=pelvis
  swing1LimitDegrees=45
  swing1Motion=Free
  swing2LimitDegrees=45
  swing2Motion=Free
  twistLimitDegrees=45
  twistMotion=Free
 }
]
