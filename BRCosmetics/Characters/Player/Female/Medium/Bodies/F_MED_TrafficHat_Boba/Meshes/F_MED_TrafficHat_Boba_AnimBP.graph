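# Post-process style AnimBlueprint for female TrafficHat_Boba body cosmetic, animating its suspenders, belt, straps and knee pieces.
# It copies the attached parent's pose, then chains Look At nodes, weighted multi-bone constraints, state-driven corrective transforms (crouch, slide, jump, parachute) and spring-based Anim Dynamics.
# Output reads cached 'BRSetup' through a front-end blend whose visible branches both use that pose.
# Several physics and transform fragments sit unconnected.
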AnimBlueprint /BRCosmetics/Characters/Player/Female/Medium/Bodies/F_MED_TrafficHat_Boba/Meshes/F_MED_TrafficHat_Boba_AnimBP
target_skeleton=&/F_MED_TrafficHat_Boba_Skeleton
graphs [
 AnimGraph {nodes [
   Anim Dynamics - Bone: dyn_pants_front_l_04(Alpha=1, ExternalForce="0,0,0") {
    alphaBoolBlend {blendOption=Linear}
    boundBone {boneName=dyn_pants_front_l_04}
    physicsBodyDefinitions [{
      boundBone {boneName=dyn_pants_front_l_04}
      boxExtents=<10 10 10>
      constraintSetup {angularLimitsMax=<15 15 15>, angularLimitsMin=<-15 -15 -15>, angularTarget=<0 0 0>}
      localJointOffset=<0 0 0>
     }]
   }
   n1 = Anim Dynamics - Bone: dyn_pants_front_l_01(ComponentPose=Local To Component(), Alpha=1, ExternalForce="0,0,0") {
    alphaBoolBlend {blendOption=Linear}
    boundBone {boneName=dyn_pants_front_l_01}
    componentAppliedLinearAccClamp=<3000 300 3000>
    componentLinearAccScale=<1 1 1>
    componentLinearVelScale=<1 1 1>
    physicsBodyDefinitions [{
      boundBone {boneName=dyn_pants_front_l_01}
      boxExtents=<10 10 10>
      constraintSetup {angularLimitsMax=<15 0 15>, angularLimitsMin=<-15 0 -15>, angularTarget=<0 0 0>}
      localJointOffset=<1 0 0>
     }]
   }
   n2 = Anim Dynamics - Bone: dyn_pants_front_l_02(ComponentPose=@n1, Alpha=1, ExternalForce="0,0,0") {
    alphaBoolBlend {blendOption=Linear}
    boundBone {boneName=dyn_pants_front_l_02}
    physicsBodyDefinitions [{
      boundBone {boneName=dyn_pants_front_l_02}
      boxExtents=<10 10 10>
      constraintSetup {angularLimitsMax=<15 0 15>, angularLimitsMin=<-15 0 -15>, angularTarget=<0 0 0>}
      localJointOffset=<1 0 0>
     }]
   }
   n3 = Anim Dynamics - Bone: dyn_pants_front_l_03(ComponentPose=@n2, Alpha=1, ExternalForce="0,0,0") {
    alphaBoolBlend {blendOption=Linear}
    boundBone {boneName=dyn_pants_front_l_03}
    physicsBodyDefinitions [{
      boundBone {boneName=dyn_pants_front_l_03}
      boxExtents=<10 10 10>
      constraintSetup {angularLimitsMax=<15 0 15>, angularLimitsMin=<-15 0 -15>, angularTarget=<0 0 0>}
      localJointOffset=<1 0 0>
     }]
   }
   n4 = Anim Dynamics - Bone: dyn_pants_front_l_04(ComponentPose=@n3, Alpha=1, ExternalForce="0,0,0") {
    alphaBoolBlend {blendOption=Linear}
    boundBone {boneName=dyn_pants_front_l_04}
    physicsBodyDefinitions [{
      boundBone {boneName=dyn_pants_front_l_04}
      boxExtents=<10 10 10>
      constraintSetup {angularLimitsMax=<15 0 15>, angularLimitsMin=<-15 0 -15>, angularTarget=<0 0 0>}
      localJointOffset=<1 0 0>
     }]
   }
   n5 = Anim Dynamics - Bone: dyn_pants_back_l_01(ComponentPose=@n4, Alpha=1, ExternalForce="0,0,0") {
    alphaBoolBlend {blendOption=Linear}
    boundBone {boneName=dyn_pants_back_l_01}
    physicsBodyDefinitions [{
      boundBone {boneName=dyn_pants_back_l_01}
      boxExtents=<10 10 10>
      constraintSetup {angularLimitsMax=<15 0 15>, angularLimitsMin=<-15 0 -15>, angularTarget=<0 0 0>}
      localJointOffset=<1 0 0>
     }]
   }
   n6 = Anim Dynamics - Bone: dyn_pants_front_l_03(ComponentPose=@n5, Alpha=1, ExternalForce="0,0,0") {
    alphaBoolBlend {blendOption=Linear}
    boundBone {boneName=dyn_pants_front_l_03}
    physicsBodyDefinitions [{
      boundBone {boneName=dyn_pants_front_l_03}
      boxExtents=<10 10 10>
      constraintSetup {angularLimitsMax=<15 0 15>, angularLimitsMin=<-15 0 -15>, angularTarget=<0 0 0>}
      localJointOffset=<1 0 0>
     }]
   }
   Anim Dynamics - Bone: dyn_pants_front_l_04(ComponentPose=@n6, Alpha=1, ExternalForce="0,0,0") {
    alphaBoolBlend {blendOption=Linear}
    boundBone {boneName=dyn_pants_front_l_04}
    physicsBodyDefinitions [{
      boundBone {boneName=dyn_pants_front_l_04}
      boxExtents=<10 10 10>
      constraintSetup {angularLimitsMax=<15 0 15>, angularLimitsMin=<-15 0 -15>, angularTarget=<0 0 0>}
      localJointOffset=<1 0 0>
     }]
   }
   Anim Dynamics - Bone: dyn_suspenders_01_f!(ComponentPose, bAlphaBoolEnabled:bool=True, ExternalForce="0,0,0") -> (Pose) {
    alphaInputType=Bool
    angularDampingOverride=1
    bOverrideAngularDamping=true
    bUseGravityOverride=true
    simulationSpace=RootRelative
    alphaBoolBlend {blendInTime=0.1, blendOption=Linear, blendOutTime=0.1}
    boundBone {boneName=dyn_suspenders_01_f}
    componentAppliedLinearAccClamp=<3000 3000 3000>
    componentLinearAccScale=<0.4 0.4 0.4>
    componentLinearVelScale=<0.5 0.5 0.5>
    gravityOverride {z=-980}
    physicsBodyDefinitions [{
      boundBone {boneName=dyn_suspenders_01_f}
      boxExtents=<10 10 10>
      constraintSetup {angularLimitsMax=<0 0 50>, angularLimitsMin=<0 -15 6>, angularTarget=<0 0 0>}
      localJointOffset=<1 0 0>
     }]
    relativeSpaceBone {boneName=pelvis}
   }
   n7 = "Transform (Modify) Bone - Bone: dyn_suspenders_01_f_MSTR"(Alpha=1, Translation="0,0,0", Rotation="-24.177,6.575,-5.653", Scale="1,1,1") {
    rotationMode=BMM_Additive
    rotationSpace=BCS_BoneSpace
    translationMode=BMM_Additive
    translationSpace=BCS_BoneSpace
    alphaBoolBlend {blendOption=Linear}
    boneToModify {boneName=dyn_suspenders_01_f_MSTR}
    rotation=<rot -24.1769 6.5752 -5.6531>
   }
   n8 = "Transform (Modify) Bone - Bone: dyn_suspenders_01_b_MSTR"(ComponentPose=@n7, Alpha=1, Translation="0,0,0", Rotation="-10.000,-0.010,0.034", Scale="1,1,1") {
    rotationMode=BMM_Additive
    rotationSpace=BCS_BoneSpace
    translationMode=BMM_Additive
    translationSpace=BCS_BoneSpace
    alphaBoolBlend {blendOption=Linear}
    boneToModify {boneName=dyn_suspenders_01_b_MSTR}
    rotation=<rot -10 -0.0096 0.0341>
   }
   n9 = Anim Dynamics - Bone: dyn_suspenders_02_s!(ComponentPose=@n8, Alpha=1, GravityOverride:Vector="0,0,-980") -> (Pose) {
    angularDampingOverride=1
    angularSpringConstant=2000
    bAngularSpring=true
    bOverrideAngularDamping=true
    bUseGravityOverride=true
    bUseSphericalLimits=true
    simulationSpace=BoneRelative
    alphaBoolBlend {blendInTime=0.3, blendOption=ExpInOut, blendOutTime=0.3}
    boundBone {boneName=dyn_suspenders_02_s}
    componentAppliedLinearAccClamp=<3000 3000 3000>
    componentLinearAccScale=<0.3 0.3 0.3>
    componentLinearVelScale=<0.3 0.3 0.3>
    gravityOverride {z=-980}
    physicsBodyDefinitions [{
      collisionType=CustomSphere
      sphereCollisionRadius=2
      boundBone {boneName=dyn_suspenders_02_s}
      boxExtents=<10 10 10>
      constraintSetup {angularLimitsMax=<0 0 15>, angularLimitsMin=<0 -15 -10>, angularTarget=<0 0 0>}
      localJointOffset=<0.5 0 0>
     }]
    relativeSpaceBone {boneName=root}
    sphericalLimits [{limitRadius=8.5, drivingBone {boneName=thigh_twist_01_l}, sphereLocalOffset=<3 -1.5 0.5>}]
   }
   Anim Dynamics - Bone: dyn_suspenders_02_s2!(ComponentPose=@n9, Alpha=1, GravityOverride:Vector="0,0,-980") -> (Pose) {
    angularDampingOverride=1
    angularSpringConstant=2000
    bAngularSpring=true
    bOverrideAngularDamping=true
    bUseGravityOverride=true
    bUseSphericalLimits=true
    simulationSpace=BoneRelative
    alphaBoolBlend {blendInTime=0.3, blendOption=ExpInOut, blendOutTime=0.3}
    boundBone {boneName=dyn_suspenders_02_s2}
    componentAppliedLinearAccClamp=<3000 3000 3000>
    componentLinearAccScale=<0.3 0.3 0.3>
    componentLinearVelScale=<0.3 0.3 0.3>
    gravityOverride {z=-980}
    physicsBodyDefinitions [{
      collisionType=CustomSphere
      sphereCollisionRadius=2
      boundBone {boneName=dyn_suspenders_02_s2}
      boxExtents=<15 15 5>
      constraintSetup {angularLimitsMax=<0 0 0>, angularLimitsMin=<0 -15 0>, angularTarget=<0 0 0>}
      localJointOffset=<0.5 0 0>
     }]
    relativeSpaceBone {boneName=root}
    sphericalLimits [
     {limitRadius=0, drivingBone {boneName=thigh_twist_01_l}, sphereLocalOffset=<6 -1.5 0.5>}
     {limitRadius=10, drivingBone {boneName=thigh_twist_01_l}, sphereLocalOffset=<2 0 0>}
    ]
   }
   n10 = Anim Dynamics - Bone: dyn_belt_r_strap_01!(ComponentPose, bAlphaBoolEnabled:bool=True, GravityOverride:Vector="0,0,-980", ExternalForce="0,0,0") -> (Pose) {
    alphaInputType=Bool
    angularDampingOverride=0.9999
    angularSpringConstant=1000
    bAngularSpring=true
    bOverrideAngularDamping=true
    bUseGravityOverride=true
    simulationSpace=RootRelative
    alphaBoolBlend {blendInTime=0.1, blendOption=Linear, blendOutTime=0.1}
    boundBone {boneName=dyn_belt_r_strap_01}
    componentAppliedLinearAccClamp=<3000 3000 3000>
    componentLinearAccScale=<0.4 0.5 0.5>
    componentLinearVelScale=<0.5 0.5 0.5>
    gravityOverride {z=-500}
    physicsBodyDefinitions [{
      collisionType=CustomSphere
      sphereCollisionRadius=2.5
      boundBone {boneName=dyn_belt_r_strap_01}
      boxExtents=<8 8 8>
      constraintSetup {angularLimitsMax=<0 5 4>, angularLimitsMin=<0 -5 -25>, angularTarget=<0 0 0>}
      localJointOffset=<2 0 0>
     }]
    relativeSpaceBone {boneName=pack}
   }
   n11 = Anim Dynamics - Bone: dyn_belt_r_strap_02!(ComponentPose=@n10, bAlphaBoolEnabled:bool=True, GravityOverride:Vector="0,0,-980", ExternalForce="0,0,0") -> (Pose) {
    alphaInputType=Bool
    angularDampingOverride=0.9999
    angularSpringConstant=1000
    bAngularSpring=true
    bOverrideAngularDamping=true
    bUseGravityOverride=true
    simulationSpace=RootRelative
    alphaBoolBlend {blendInTime=0.1, blendOption=Linear, blendOutTime=0.1}
    boundBone {boneName=dyn_belt_r_strap_02}
    componentAppliedLinearAccClamp=<3000 3000 3000>
    componentLinearAccScale=<0.4 0.5 0.5>
    componentLinearVelScale=<0.5 0.5 0.5>
    gravityOverride {z=-500}
    physicsBodyDefinitions [{
      collisionType=CustomSphere
      sphereCollisionRadius=2.5
      boundBone {boneName=dyn_belt_r_strap_02}
      boxExtents=<8 8 8>
      constraintSetup {angularLimitsMax=<0 5 4>, angularLimitsMin=<0 -5 -25>, angularTarget=<0 0 0>}
      localJointOffset=<2 0 0>
     }]
    relativeSpaceBone {boneName=pack}
   }
   Anim Dynamics - Bone: dyn_belt_r_strap_03!(ComponentPose=@n11, bAlphaBoolEnabled:bool=True, GravityOverride:Vector="0,0,-980", ExternalForce="0,0,0") -> (Pose) {
    alphaInputType=Bool
    angularDampingOverride=0.9999
    angularSpringConstant=1000
    bAngularSpring=true
    bOverrideAngularDamping=true
    bUseGravityOverride=true
    simulationSpace=RootRelative
    alphaBoolBlend {blendInTime=0.1, blendOption=Linear, blendOutTime=0.1}
    boundBone {boneName=dyn_belt_r_strap_03}
    componentAppliedLinearAccClamp=<3000 3000 3000>
    componentLinearAccScale=<0.4 0.5 0.5>
    componentLinearVelScale=<0.5 0.5 0.5>
    gravityOverride {z=-500}
    physicsBodyDefinitions [{
      collisionType=CustomSphere
      sphereCollisionRadius=2.5
      boundBone {boneName=dyn_belt_r_strap_03}
      boxExtents=<8 8 8>
      constraintSetup {angularLimitsMax=<0 5 4>, angularLimitsMin=<0 -5 -25>, angularTarget=<0 0 0>}
      localJointOffset=<2 0 0>
     }]
    relativeSpaceBone {boneName=pack}
   }
   "Constraint - dyn_belt_l2_01_MSTR (PPP)"!(ComponentPose, Alpha=1, ConstraintWeights_0=0.8869, ConstraintWeights_1:float=0.1129, ConstraintWeights_2:float=0.0002) -> (Pose) {
    alphaBoolBlend {blendOption=Linear}
    boneToModify {boneName=dyn_belt_l2_01_MSTR}
    constraintSetup [{targetBone {boneName=pelvis}}, {targetBone {boneName=thigh_twist_01_l}}, {targetBone {boneName=thigh_l}}]
    constraintWeights [1,1,1]
   }
   Look At(Alpha=1) {alphaBoolBlend {blendOption=Linear}, boneToModify=~, lookAtTarget=~}
   n12 = Look At - Bone: dyn_suspenders_01_s_MSTR(Alpha=1) {
    alphaBoolBlend {blendOption=Linear}
    boneToModify {boneName=dyn_suspenders_01_s_MSTR}
    lookAtLocation=<-46.9646 7.601 -17.0179>
    lookAtTarget {boneReference {boneName=thigh_l}}
    lookAt_Axis {axis=<1 0>}
   }
   Look At - Bone: dyn_suspenders_01_s2_MSTR(ComponentPose=@n12, Alpha=1) {
    alphaBoolBlend {blendOption=Linear}
    boneToModify {boneName=dyn_suspenders_01_s2_MSTR}
    lookAtLocation=<-46.9646 7.601 -17.0179>
    lookAtTarget {boneReference {boneName=thigh_l}}
    lookAt_Axis {axis=<1 0>}
   }
   "Transform (Modify) Bone - Bone: dyn_pants_front_l_01_MSTR"(Alpha=1, Translation="0,10,0", Rotation="90,0,0", Scale="1,1,1") {
    rotationMode=BMM_Additive
    rotationSpace=BCS_ParentBoneSpace
    scaleMode=BMM_Additive
    scaleSpace=BCS_BoneSpace
    alphaBoolBlend {blendOption=Linear}
    boneToModify {boneName=dyn_pants_front_l_01_MSTR}
    rotation {pitch=90}
    translation {y=10}
   }
   n13 = Get Emote() -> (Emote:double)
   n14 = "Transform (Modify) Bone - Bone: dyn_suspenders_01_f_MSTR"(Alpha=@n13, Translation="0,5,0", Rotation="45,10,10", Scale="1,1,1") {
    rotationMode=BMM_Additive
    rotationSpace=BCS_BoneSpace
    translationMode=BMM_Additive
    translationSpace=BCS_BoneSpace
    alphaBoolBlend {blendOption=Linear}
    boneToModify {boneName=dyn_suspenders_01_f_MSTR}
    rotation=<rot 45 10 10>
    translation {y=5}
   }
   n15 = "Transform (Modify) Bone - Bone: dyn_suspenders_01_s_MSTR"(ComponentPose=@n14, Alpha=@n13, Translation="-5,0,5", Rotation="10,0,0", Scale="1,1,1") {
    rotationMode=BMM_Additive
    rotationSpace=BCS_BoneSpace
    translationMode=BMM_Additive
    translationSpace=BCS_BoneSpace
    alphaBoolBlend {blendOption=Linear}
    boneToModify {boneName=dyn_suspenders_01_s_MSTR}
    rotation {pitch=10}
    translation {x=-5, z=5}
   }
   "Transform (Modify) Bone - Bone: dyn_suspenders_01_s2_MSTR"(ComponentPose=@n15, Alpha=@n13, Translation="0,5,5", Rotation="0,0,0", Scale="1,1,1") {
    rotationMode=BMM_Additive
    rotationSpace=BCS_BoneSpace
    translationMode=BMM_Additive
    translationSpace=BCS_BoneSpace
    alphaBoolBlend {blendOption=Linear}
    boneToModify {boneName=dyn_suspenders_01_s2_MSTR}
    translation {y=5, z=5}
   }
   n16 = RigidBody() {
    overridePhysicsAsset=&/F_MED_TrafficHat_Boba_Physics
    simulationSpace=BaseBoneSpace
    alphaBoolBlend {blendOption=Linear}
   }
   n17 = Get bIsCrouching()
   n18 = "Transform (Modify) Bone - Bone: dyn_suspenders_01_l"!(ComponentPose=@n16, bAlphaBoolEnabled:bool=@n17, Translation="0,0,0", Rotation="-0.020,0.027,-50.000", Scale="1,1,1") -> (Pose) {
    alphaInputType=Bool
    rotationMode=BMM_Additive
    rotationSpace=BCS_BoneSpace
    scaleSpace=BCS_BoneSpace
    translationMode=BMM_Additive
    translationSpace=BCS_BoneSpace
    alphaBoolBlend {blendOption=Linear}
    boneToModify {boneName=dyn_suspenders_01_l}
    rotation=<rot -0.02 0.027 -50>
   }
   "Transform (Modify) Bone - Bone: dyn_suspenders_02_l"!(ComponentPose=@n18, bAlphaBoolEnabled:bool=@n17, Translation="0,0,0", Rotation="20,0,30", Scale="1,1,1") -> (Pose) {
    alphaInputType=Bool
    rotationMode=BMM_Additive
    rotationSpace=BCS_BoneSpace
    scaleSpace=BCS_BoneSpace
    translationMode=BMM_Additive
    translationSpace=BCS_BoneSpace
    alphaBoolBlend {blendOption=Linear}
    boneToModify {boneName=dyn_suspenders_02_l}
    rotation {pitch=20, roll=30}
   }
   "Transform (Modify) Bone - Bone: dyn_suspenders_01_skin_l"!(ComponentPose, bAlphaBoolEnabled:bool=True, Translation="1.494,-0.317,2.036", Rotation="12.686,-45.590,-96.747", Scale="1,1,1") -> (Pose) {
    alphaInputType=Bool
    rotationMode=BMM_Additive
    rotationSpace=BCS_BoneSpace
    scaleMode=BMM_Additive
    scaleSpace=BCS_BoneSpace
    translationMode=BMM_Additive
    translationSpace=BCS_BoneSpace
    alphaBoolBlend {blendInTime=0.1, blendOption=Linear, blendOutTime=0.1}
    boneToModify {boneName=dyn_suspenders_01_skin_l}
    rotation=<rot 12.686 -45.59 -96.747>
    translation=<1.494 -0.317 2.036>
   }
   RigidBody!(ComponentPose, Alpha, OverrideWorldGravity:Vector="0,0,0", ExternalForce) -> (Pose) {
    bOverrideWorldGravity=true
    overridePhysicsAsset=&F_MED_TrafficHat_Boba_Physics
    simulationSpace=BaseBoneSpace
    alphaBoolBlend {blendOption=Linear}
    baseBoneRef {boneName=pelvis}
    componentAppliedLinearAccClamp=<1500 1500 1500>
    componentLinearAccScale=<0.3 0.3 0.4>
    componentLinearVelScale=<0.2 0.2 0.3>
   }
   n19 = Use cached pose 'BRSetup'()
   n20 = Blend Poses by bool(
    BlendPose_0=Component To Local(ComponentPose=Local To Component(LocalPose=@n19))
    BlendPose_1=Use cached pose 'BRSetup'()
    bActiveValue=Get bIsInFrontEnd()
   ) {blendType=Linear}
   Output Pose(Result=Component To Local(ComponentPose=Local To Component(LocalPose=@n20))) {layerGroup=None}
   n21 = Get Crouching_Not_sliding() -> (Crouching_Not_sliding:bool)
   n22 = Get LookAt_Alpha() -> (LookAt_Alpha:double)
   n23 = Look At - Bone: dyn_suspenders_01_s_MSTR(ComponentPose=Local To Component(LocalPose=Copy Pose From Mesh() {bUseAttachedParent=true}), Alpha=@n22) {
    alphaBoolBlend {blendOption=Linear}
    boneToModify {boneName=dyn_suspenders_01_s_MSTR}
    lookAtLocation=<-41.771 7.601 -17.0179>
    lookAtTarget {boneReference {boneName=thigh_l}}
    lookAt_Axis {axis=<1 0>}
   }
   n24 = Look At - Bone: dyn_suspenders_01_s2_MSTR(ComponentPose=@n23, Alpha=@n22) {
    alphaBoolBlend {blendOption=Linear}
    boneToModify {boneName=dyn_suspenders_01_s2_MSTR}
    lookAtLocation=<-47.8403 7.601 -17.0178>
    lookAtTarget {boneReference {boneName=thigh_l}}
    lookAt_Axis {axis=<1 0>}
   }
   n25 = Look At - Bone: dyn_suspenders_01_f_MSTR(ComponentPose=@n24, Alpha=@n22) {
    alphaBoolBlend {blendOption=Linear}
    boneToModify {boneName=dyn_suspenders_01_f_MSTR}
    lookAtLocation=<-47.7193 -8.8758 -25.1321>
    lookAtTarget {boneReference {boneName=thigh_l}}
    lookAt_Axis {axis=<1 0>}
   }
   n26 = "Constraint - dyn_suspenders_r_01_top_MSTR (PPPP)"!(
    ComponentPose=@n25
    Alpha=1
    ConstraintWeights_0=0.011
    ConstraintWeights_1:float=0.4895
    ConstraintWeights_2:float=0.485
    ConstraintWeights_3:float=0.0145
   ) -> (Pose) {
    alphaBoolBlend {blendOption=Linear}
    boneToModify {boneName=dyn_suspenders_r_01_top_MSTR}
    constraintSetup [
     {targetBone {boneName=spine_04}}
     {targetBone {boneName=spine_05}}
     {targetBone {boneName=clavicle_r}}
     {targetBone {boneName=upperarm_twist_01_r}}
    ]
    constraintWeights [1,1,1,1]
   }
   n27 = "Constraint - dyn_suspenders_r_01_MSTR (PPPP)"!(
    ComponentPose=@n26
    Alpha=1
    ConstraintWeights_0=0.2918
    ConstraintWeights_1:float=0.3499
    ConstraintWeights_2:float=0.344
    ConstraintWeights_3:float=0.0143
   ) -> (Pose) {
    alphaBoolBlend {blendOption=Linear}
    boneToModify {boneName=dyn_suspenders_r_01_MSTR}
    constraintSetup [
     {targetBone {boneName=spine_04}}
     {targetBone {boneName=spine_05}}
     {targetBone {boneName=clavicle_r}}
     {targetBone {boneName=upperarm_twist_01_r}}
    ]
    constraintWeights [1,1,1,1]
   }
   n28 = "Constraint - dyn_knee_l_01_MSTR (PPP)"!(ComponentPose=@n27, Alpha=1, ConstraintWeights_0=0.1929, ConstraintWeights_1:float=0.8061, ConstraintWeights_2:float=0.001) -> (Pose) {
    alphaBoolBlend {blendOption=Linear}
    boneToModify {boneName=dyn_knee_l_01_MSTR}
    constraintSetup [{targetBone {boneName=thigh_l}}, {targetBone {boneName=calf_l}}, {targetBone {boneName=calf_twist_02_l}}]
    constraintWeights [1,1,1]
   }
   n29 = "Constraint - dyn_knee_r_01_MSTR (PPP)"!(ComponentPose=@n28, Alpha=1, ConstraintWeights_0=0.2318, ConstraintWeights_1:float=0.765, ConstraintWeights_2:float=0.0031) -> (Pose) {
    alphaBoolBlend {blendOption=Linear}
    boneToModify {boneName=dyn_knee_r_01_MSTR}
    constraintSetup [{targetBone {boneName=thigh_r}}, {targetBone {boneName=calf_r}}, {targetBone {boneName=calf_twist_02_r}}]
    constraintWeights [1,1,1]
   }
   n30 = "Constraint - dyn_belt_r_01_MSTR (PPP)"!(ComponentPose=@n29, Alpha=1, ConstraintWeights_0=0.3111, ConstraintWeights_1:float=0.5703, ConstraintWeights_2:float=0.1186) -> (Pose) {
    alphaBoolBlend {blendOption=Linear}
    boneToModify {boneName=dyn_belt_r_01_MSTR}
    constraintSetup [{targetBone {boneName=pelvis}}, {targetBone {boneName=thigh_twist_01_r}}, {targetBone {boneName=thigh_r}}]
    constraintWeights [1,1,1]
   }
   n31 = "Constraint - dyn_belt_ring_01_MSTR (PPPPPPPP)"!(
    ComponentPose=@n30
    Alpha=1
    ConstraintWeights_0=0.0004
    ConstraintWeights_1:float=0.0002
    ConstraintWeights_2:float=0.0002
    ConstraintWeights_3:float=0.2617
    ConstraintWeights_4:float=0.0037
    ConstraintWeights_5:float=0.6568
    ConstraintWeights_6:float=0.0766
    ConstraintWeights_7:float=0.0004
   ) -> (Pose) {
    alphaBoolBlend {blendOption=Linear}
    boneToModify {boneName=dyn_belt_ring_01_MSTR}
    constraintSetup [
     {targetBone {boneName=spine_02}}
     {targetBone {boneName=spine_04}}
     {targetBone {boneName=spine_01}}
     {targetBone {boneName=pelvis}}
     {targetBone {boneName=thigh_twist_01_l}}
     {targetBone {boneName=thigh_twist_01_r}}
     {targetBone {boneName=thigh_r}}
     {targetBone {boneName=thigh_l}}
    ]
    constraintWeights [1,1,1,1,1,1,1,1]
   }
   n32 = "Constraint - dyn_belt_c_01_MSTR (PPPPPPPP)"!(
    ComponentPose=@n31
    Alpha=1
    ConstraintWeights_0=0.003
    ConstraintWeights_1:float=0.0013
    ConstraintWeights_2:float=0.0017
    ConstraintWeights_3:float=0.8191
    ConstraintWeights_4:float=0.0172
    ConstraintWeights_5:float=0.1495
    ConstraintWeights_6:float=0.0069
    ConstraintWeights_7:float=0.0012
   ) -> (Pose) {
    alphaBoolBlend {blendOption=Linear}
    boneToModify {boneName=dyn_belt_c_01_MSTR}
    constraintSetup [
     {targetBone {boneName=spine_02}}
     {targetBone {boneName=spine_04}}
     {targetBone {boneName=spine_01}}
     {targetBone {boneName=pelvis}}
     {targetBone {boneName=thigh_twist_01_l}}
     {targetBone {boneName=thigh_twist_01_r}}
     {targetBone {boneName=thigh_r}}
     {targetBone {boneName=thigh_l}}
    ]
    constraintWeights [1,1,1,1,1,1,1,1]
   }
   n33 = "Constraint - dyn_belt_l_01_MSTR (PPPPPPP)"!(
    ComponentPose=@n32
    Alpha=1
    ConstraintWeights_0=0.0003
    ConstraintWeights_1:float=0.0002
    ConstraintWeights_2:float=0.794
    ConstraintWeights_3:float=0.1548
    ConstraintWeights_4:float=0.0311
    ConstraintWeights_5:float=0.0041
    ConstraintWeights_6:float=0.0156
   ) -> (Pose) {
    alphaBoolBlend {blendOption=Linear}
    boneToModify {boneName=dyn_belt_l_01_MSTR}
    constraintSetup [
     {targetBone {boneName=spine_02}}
     {targetBone {boneName=spine_01}}
     {targetBone {boneName=pelvis}}
     {targetBone {boneName=thigh_twist_01_l}}
     {targetBone {boneName=thigh_twist_01_r}}
     {targetBone {boneName=thigh_r}}
     {targetBone {boneName=thigh_l}}
    ]
    constraintWeights [1,1,1,1,1,1,1]
   }
   n34 = "Transform (Modify) Bone - Bone: dyn_belt_r_01_MSTR"!(ComponentPose=@n33, bAlphaBoolEnabled:bool=@n21, Translation="-4.948,-0.514,0.009", Rotation="-29.577,-11.473,4.840", Scale="1,1,1") -> (Pose) {
    alphaInputType=Bool
    rotationMode=BMM_Additive
    rotationSpace=BCS_BoneSpace
    scaleSpace=BCS_BoneSpace
    translationMode=BMM_Additive
    translationSpace=BCS_BoneSpace
    alphaBoolBlend {blendInTime=0.1, blendOption=Linear, blendOutTime=0.1}
    boneToModify {boneName=dyn_belt_r_01_MSTR}
    rotation=<rot -29.5768 -11.473 4.84>
    translation=<-4.948 -0.514 0.009>
   }
   n35 = "Transform (Modify) Bone - Bone: dyn_belt_ring_01_MSTR"!(ComponentPose=@n34, bAlphaBoolEnabled:bool=@n21, Translation="1.502,2.134,-2.680", Rotation="-25.000,0.000,0.000", Scale="1,1,1") -> (Pose) {
    alphaInputType=Bool
    rotationMode=BMM_Additive
    rotationSpace=BCS_BoneSpace
    scaleSpace=BCS_BoneSpace
    translationMode=BMM_Additive
    translationSpace=BCS_BoneSpace
    alphaBoolBlend {blendInTime=0.1, blendOption=Linear, blendOutTime=0.1}
    boneToModify {boneName=dyn_belt_ring_01_MSTR}
    rotation=<rot -25 0 0>
    translation=<1.502 2.134 -2.68>
   }
   n36 = "Transform (Modify) Bone - Bone: dyn_belt_l_01"!(ComponentPose=@n35, bAlphaBoolEnabled:bool=@n21, Translation="-2.5,3.025,0", Rotation="0,0,0", Scale="1,1,1") -> (Pose) {
    alphaInputType=Bool
    rotationMode=BMM_Additive
    rotationSpace=BCS_BoneSpace
    scaleMode=BMM_Additive
    scaleSpace=BCS_BoneSpace
    translationMode=BMM_Additive
    translationSpace=BCS_BoneSpace
    alphaBoolBlend {blendInTime=0.1, blendOption=Linear, blendOutTime=0.1}
    boneToModify {boneName=dyn_belt_l_01}
    translation=<-2.5 3.025>
   }
   n37 = "Transform (Modify) Bone - Bone: dyn_belt_c_01"!(ComponentPose=@n36, bAlphaBoolEnabled:bool=@n21, Translation="-8,3,0", Rotation="0,0,0", Scale="1,1,1") -> (Pose) {
    alphaInputType=Bool
    rotationMode=BMM_Additive
    rotationSpace=BCS_BoneSpace
    scaleSpace=BCS_BoneSpace
    translationMode=BMM_Additive
    translationSpace=BCS_BoneSpace
    alphaBoolBlend {blendInTime=0.1, blendOption=Linear, blendOutTime=0.1}
    boneToModify {boneName=dyn_belt_c_01}
    translation=<-8 3>
   }
   n38 = "Transform (Modify) Bone - Bone: dyn_belt_l_02"!(ComponentPose=@n37, bAlphaBoolEnabled:bool=@n21, Translation="0,0,0", Rotation="-40,0,0", Scale="1,1,1") -> (Pose) {
    alphaInputType=Bool
    rotationMode=BMM_Additive
    rotationSpace=BCS_BoneSpace
    scaleSpace=BCS_BoneSpace
    translationMode=BMM_Additive
    translationSpace=BCS_BoneSpace
    alphaBoolBlend {blendInTime=0.1, blendOption=Linear, blendOutTime=0.1}
    boneToModify {boneName=dyn_belt_l_02}
    rotation {pitch=-40}
   }
   n39 = "Transform (Modify) Bone - Bone: dyn_belt_l2_01"!(ComponentPose=@n38, bAlphaBoolEnabled:bool=@n21, Translation="-4.214,2.387,-0.492", Rotation="-23.399,13.118,9.062", Scale="1,1,1") -> (Pose) {
    alphaInputType=Bool
    rotationMode=BMM_Additive
    rotationSpace=BCS_BoneSpace
    scaleMode=BMM_Additive
    scaleSpace=BCS_BoneSpace
    translationMode=BMM_Additive
    translationSpace=BCS_BoneSpace
    alphaBoolBlend {blendInTime=0.1, blendOption=Linear, blendOutTime=0.1}
    boneToModify {boneName=dyn_belt_l2_01}
    rotation=<rot -23.3991 13.1181 9.062>
    translation=<-4.214 2.387 -0.492>
   }
   n40 = "Transform (Modify) Bone - Bone: dyn_suspenders_01_b"!(
    ComponentPose=@n39
    bAlphaBoolEnabled:bool=Get Crouching_backpedaling() -> (Crouching_backpedaling:bool)
    Translation="0,-5,0"
    Rotation="0,0,0"
    Scale="1,1,1"
   ) -> (Pose) {
    alphaInputType=Bool
    rotationMode=BMM_Additive
    rotationSpace=BCS_BoneSpace
    scaleMode=BMM_Additive
    scaleSpace=BCS_BoneSpace
    translationMode=BMM_Additive
    translationSpace=BCS_BoneSpace
    alphaBoolBlend {blendInTime=0.1, blendOption=Linear, blendOutTime=0.1}
    boneToModify {boneName=dyn_suspenders_01_b}
    translation {y=-5}
   }
   n41 = "Transform (Modify) Bone - Bone: dyn_suspenders_01_f_MSTR"!(ComponentPose=@n40, bAlphaBoolEnabled:bool=Get bIsTacticalSprinting(), Translation="0,0,0", Rotation="0,10,0", Scale="1,1,1") -> (Pose) {
    alphaInputType=Bool
    rotationMode=BMM_Additive
    rotationSpace=BCS_BoneSpace
    scaleMode=BMM_Additive
    scaleSpace=BCS_BoneSpace
    translationMode=BMM_Additive
    translationSpace=BCS_BoneSpace
    alphaBoolBlend {blendInTime=0.1, blendOption=Linear, blendOutTime=0.1}
    boneToModify {boneName=dyn_suspenders_01_f_MSTR}
    rotation {yaw=10}
   }
   n42 = "Transform (Modify) Bone - Bone: dyn_belt_r_strap_01"!(ComponentPose=@n41, bAlphaBoolEnabled:bool=True, Translation="0,0,0", Rotation="-14.922,-5.186,0.444", Scale="1,1,1") -> (Pose) {
    alphaInputType=Bool
    rotationMode=BMM_Additive
    rotationSpace=BCS_BoneSpace
    scaleMode=BMM_Additive
    scaleSpace=BCS_BoneSpace
    translationMode=BMM_Additive
    translationSpace=BCS_BoneSpace
    alphaBoolBlend {blendInTime=0.1, blendOption=Linear, blendOutTime=0.1}
    boneToModify {boneName=dyn_belt_r_strap_01}
    rotation=<rot -14.922 -5.186 0.444>
   }
   n43 = "Transform (Modify) Bone - Bone: dyn_belt_r_strap_02"!(ComponentPose=@n42, bAlphaBoolEnabled:bool=True, Translation="0,0,0", Rotation="-0.005,-0.085,0.020", Scale="1,1,1") -> (Pose) {
    alphaInputType=Bool
    rotationMode=BMM_Additive
    rotationSpace=BCS_BoneSpace
    scaleMode=BMM_Additive
    scaleSpace=BCS_BoneSpace
    translationMode=BMM_Additive
    translationSpace=BCS_BoneSpace
    alphaBoolBlend {blendInTime=0.1, blendOption=Linear, blendOutTime=0.1}
    boneToModify {boneName=dyn_belt_r_strap_02}
    rotation=<rot -0.005 -0.085 0.02>
   }
   n44 = "Transform (Modify) Bone - Bone: dyn_belt_r_strap_03"!(ComponentPose=@n43, bAlphaBoolEnabled:bool=True, Translation="0,0,0", Rotation="14.779,-2.026,2.964", Scale="1,1,1") -> (Pose) {
    alphaInputType=Bool
    rotationMode=BMM_Additive
    rotationSpace=BCS_BoneSpace
    scaleMode=BMM_Additive
    scaleSpace=BCS_BoneSpace
    translationMode=BMM_Additive
    translationSpace=BCS_BoneSpace
    alphaBoolBlend {blendInTime=0.1, blendOption=Linear, blendOutTime=0.1}
    boneToModify {boneName=dyn_belt_r_strap_03}
    rotation=<rot 14.779 -2.026 2.964>
   }
   n45 = Get Crouching Moving() -> (Crouching Moving:double)
   n46 = "Transform (Modify) Bone - Bone: dyn_belt_ring_01"(ComponentPose=@n44, Alpha=@n45, Translation="0,0,0", Rotation="0,0,0", Scale="0.8,0.8,0.8") {
    rotationMode=BMM_Additive
    rotationSpace=BCS_BoneSpace
    scaleMode=BMM_Additive
    scaleSpace=BCS_BoneSpace
    translationMode=BMM_Additive
    translationSpace=BCS_BoneSpace
    alphaBoolBlend {blendInTime=0.1, blendOption=Linear, blendOutTime=0.1}
    boneToModify {boneName=dyn_belt_ring_01}
    scale=<0.8 0.8 0.8>
   }
   n47 = "Transform (Modify) Bone - Bone: dyn_belt_l_01"!(ComponentPose=@n46, Alpha:float=@n45, Translation="0,0,0", Rotation="0,0,0", Scale="0.8,0.8,0.8") -> (Pose) {
    rotationMode=BMM_Additive
    rotationSpace=BCS_BoneSpace
    scaleMode=BMM_Additive
    scaleSpace=BCS_BoneSpace
    translationMode=BMM_Additive
    translationSpace=BCS_BoneSpace
    alphaBoolBlend {blendInTime=0.1, blendOption=Linear, blendOutTime=0.1}
    boneToModify {boneName=dyn_belt_l_01}
    scale=<0.8 0.8 0.8>
   }
   n48 = Get bIsKneeSliding()
   n49 = "Transform (Modify) Bone - Bone: dyn_belt_r_01_MSTR"(ComponentPose=@n47, bAlphaBoolEnabled=@n48, Translation="-1.323,0.000,0.000", Rotation="0,0,0", Scale="1,1,1") {
    alphaInputType=Bool
    rotationMode=BMM_Additive
    rotationSpace=BCS_BoneSpace
    scaleMode=BMM_Additive
    scaleSpace=BCS_BoneSpace
    translationMode=BMM_Additive
    translationSpace=BCS_BoneSpace
    alphaBoolBlend {blendInTime=0.1, blendOption=Linear, blendOutTime=0.1}
    boneToModify {boneName=dyn_belt_r_01_MSTR}
    translation {x=-1.323}
   }
   n50 = "Transform (Modify) Bone - Bone: dyn_belt_r_strap_01"(ComponentPose=@n49, bAlphaBoolEnabled=@n48, Translation="0,0,0", Rotation="-45,0,0", Scale="1,1,1") {
    alphaInputType=Bool
    rotationMode=BMM_Additive
    rotationSpace=BCS_BoneSpace
    scaleMode=BMM_Additive
    scaleSpace=BCS_BoneSpace
    translationMode=BMM_Additive
    translationSpace=BCS_BoneSpace
    alphaBoolBlend {blendInTime=0.1, blendOption=Linear, blendOutTime=0.1}
    boneToModify {boneName=dyn_belt_r_strap_01}
    rotation {pitch=-45}
   }
   n51 = "Transform (Modify) Bone - Bone: dyn_belt_r_strap_02"(ComponentPose=@n50, bAlphaBoolEnabled=@n48, Translation="0,0,0", Rotation="-20,0,0", Scale="1,1,1") {
    alphaInputType=Bool
    rotationMode=BMM_Additive
    rotationSpace=BCS_BoneSpace
    scaleMode=BMM_Additive
    scaleSpace=BCS_BoneSpace
    translationMode=BMM_Additive
    translationSpace=BCS_BoneSpace
    alphaBoolBlend {blendInTime=0.1, blendOption=Linear, blendOutTime=0.1}
    boneToModify {boneName=dyn_belt_r_strap_02}
    rotation {pitch=-20}
   }
   n52 = "Transform (Modify) Bone - Bone: dyn_belt_r_strap_03"(ComponentPose=@n51, bAlphaBoolEnabled=@n48, Translation="0,0,0", Rotation="0,0,0", Scale="1,1,1") {
    alphaInputType=Bool
    rotationMode=BMM_Additive
    rotationSpace=BCS_BoneSpace
    scaleMode=BMM_Additive
    scaleSpace=BCS_BoneSpace
    translationMode=BMM_Additive
    translationSpace=BCS_BoneSpace
    alphaBoolBlend {blendInTime=0.1, blendOption=Linear, blendOutTime=0.1}
    boneToModify {boneName=dyn_belt_r_strap_03}
   }
   n53 = Get bIsJumping()
   n54 = "Transform (Modify) Bone - Bone: dyn_belt_c_01"(ComponentPose=@n52, bAlphaBoolEnabled=@n53, Translation="-2.697,2.945,-0.090", Rotation="0,0,0", Scale="1,1,1") {
    alphaInputType=Bool
    rotationMode=BMM_Additive
    rotationSpace=BCS_BoneSpace
    scaleMode=BMM_Additive
    scaleSpace=BCS_BoneSpace
    translationMode=BMM_Additive
    translationSpace=BCS_BoneSpace
    alphaBoolBlend {blendInTime=0.3, blendOption=Linear, blendOutTime=0.8}
    boneToModify {boneName=dyn_belt_c_01}
    translation=<-2.697 2.945 -0.09>
   }
   n55 = "Transform (Modify) Bone - Bone: dyn_belt_l_01"!(ComponentPose=@n54, bAlphaBoolEnabled:bool=@n53, Translation="-2,0,0", Rotation="0,0,0", Scale="1,1,1") -> (Pose) {
    alphaInputType=Bool
    rotationMode=BMM_Additive
    rotationSpace=BCS_BoneSpace
    scaleMode=BMM_Additive
    scaleSpace=BCS_BoneSpace
    translationMode=BMM_Additive
    translationSpace=BCS_BoneSpace
    alphaBoolBlend {blendInTime=0.3, blendOption=Linear, blendOutTime=0.8}
    boneToModify {boneName=dyn_belt_l_01}
    translation {x=-2}
   }
   n56 = "Transform (Modify) Bone - Bone: dyn_belt_l_02"(ComponentPose=@n55, bAlphaBoolEnabled=@n53, Translation="0,0,0", Rotation="-40,0,0", Scale="1,1,1") {
    alphaInputType=Bool
    rotationMode=BMM_Additive
    rotationSpace=BCS_BoneSpace
    scaleMode=BMM_Additive
    scaleSpace=BCS_BoneSpace
    translationMode=BMM_Additive
    translationSpace=BCS_BoneSpace
    alphaBoolBlend {blendInTime=0.3, blendOption=Linear, blendOutTime=0.8}
    boneToModify {boneName=dyn_belt_l_02}
    rotation {pitch=-40}
   }
   n57 = "Transform (Modify) Bone - Bone: dyn_belt_ring_01"!(ComponentPose=@n56, bAlphaBoolEnabled:bool=@n53, Translation="0,1,0", Rotation="-20,0,0", Scale="1,1,1") -> (Pose) {
    alphaInputType=Bool
    rotationMode=BMM_Additive
    rotationSpace=BCS_BoneSpace
    scaleMode=BMM_Additive
    scaleSpace=BCS_BoneSpace
    translationMode=BMM_Additive
    translationSpace=BCS_BoneSpace
    alphaBoolBlend {blendInTime=0.3, blendOption=Linear, blendOutTime=0.8}
    boneToModify {boneName=dyn_belt_ring_01}
    rotation {pitch=-20}
    translation {y=1}
   }
   n58 = "Transform (Modify) Bone - Bone: dyn_suspenders_r_01_top"!(ComponentPose=@n57, bAlphaBoolEnabled:bool=Get bIsParachuteOpened(), Translation="0,0,-2.2", Rotation="-25,0,0", Scale="1,1,1") -> (Pose) {
    alphaInputType=Bool
    rotationMode=BMM_Additive
    rotationSpace=BCS_BoneSpace
    scaleMode=BMM_Additive
    scaleSpace=BCS_BoneSpace
    translationMode=BMM_Additive
    translationSpace=BCS_BoneSpace
    alphaBoolBlend {blendInTime=0.3, blendOption=Linear, blendOutTime=0.8}
    boneToModify {boneName=dyn_suspenders_r_01_top}
    rotation {pitch=-25}
    translation {z=-2.2}
   }
   n59 = Get bIsKneeSliding()
   n60 = "Transform (Modify) Bone - Bone: dyn_suspenders_01_b"(ComponentPose=@n58, bAlphaBoolEnabled=@n59, Translation="0,0,0", Rotation="50.000,-65.000,-0.000", Scale="1,1,1") {
    alphaInputType=Bool
    rotationMode=BMM_Additive
    rotationSpace=BCS_BoneSpace
    scaleMode=BMM_Additive
    scaleSpace=BCS_BoneSpace
    translationMode=BMM_Additive
    translationSpace=BCS_BoneSpace
    alphaBoolBlend {blendInTime=0.1, blendOption=Linear, blendOutTime=0.1}
    boneToModify {boneName=dyn_suspenders_01_b}
    rotation=<rot 50 -65 0>
   }
   n61 = "Transform (Modify) Bone - Bone: dyn_suspenders_02_s2"!(ComponentPose=@n60, bAlphaBoolEnabled:bool=@n59, Translation="-0.000,0.000,10.503", Rotation="0,0,0", Scale="1,1,1") -> (Pose) {
    alphaInputType=Bool
    rotationMode=BMM_Additive
    rotationSpace=BCS_BoneSpace
    scaleMode=BMM_Additive
    scaleSpace=BCS_BoneSpace
    translationMode=BMM_Additive
    translationSpace=BCS_BoneSpace
    alphaBoolBlend {blendInTime=0.1, blendOption=Linear, blendOutTime=0.1}
    boneToModify {boneName=dyn_suspenders_02_s2}
    translation {z=10.503}
   }
   n62 = Anim Dynamics - Bone: dyn_knee_l_01!(ComponentPose=@n61, bAlphaBoolEnabled:bool=True, ExternalForce="0,0,0") -> (Pose) {
    alphaInputType=Bool
    angularDampingOverride=1
    angularSpringConstant=1000
    bAngularSpring=true
    bOverrideAngularDamping=true
    bUseGravityOverride=true
    bUsePlanarLimit=false
    simulationSpace=RootRelative
    alphaBoolBlend {blendInTime=0.1, blendOption=Linear, blendOutTime=0.1}
    boundBone {boneName=dyn_knee_l_01}
    componentAppliedLinearAccClamp=<3000 3000 3000>
    componentLinearAccScale=<0.4 0.5 0.5>
    componentLinearVelScale=<0.5 0.5 0.5>
    gravityOverride {z=-980}
    physicsBodyDefinitions [{
      collisionType=CustomSphere
      sphereCollisionRadius=2.5
      boundBone {boneName=dyn_knee_l_01}
      boxExtents=<8 4 8>
      constraintSetup {angularLimitsMax=<0 25 10>, angularLimitsMin=<0 -5 -10>, angularTarget=<0 0 0>}
      localJointOffset=<2 0 0>
     }]
    relativeSpaceBone {boneName=pelvis}
   }
   n63 = Anim Dynamics - Bone: dyn_knee_r_01!(ComponentPose=@n62, bAlphaBoolEnabled:bool=True, ExternalForce="0,0,0") -> (Pose) {
    alphaInputType=Bool
    angularDampingOverride=1
    angularSpringConstant=1000
    bAngularSpring=true
    bOverrideAngularDamping=true
    bUseGravityOverride=true
    bUsePlanarLimit=false
    simulationSpace=RootRelative
    alphaBoolBlend {blendInTime=0.1, blendOption=Linear, blendOutTime=0.1}
    boundBone {boneName=dyn_knee_r_01}
    componentAppliedLinearAccClamp=<3000 3000 3000>
    componentLinearAccScale=<0.4 0.5 0.5>
    componentLinearVelScale=<0.5 0.5 0.5>
    gravityOverride {z=-980}
    physicsBodyDefinitions [{
      collisionType=CustomSphere
      sphereCollisionRadius=2.5
      boundBone {boneName=dyn_knee_r_01}
      boxExtents=<8 4 8>
      constraintSetup {angularLimitsMax=<0 25 10>, angularLimitsMin=<0 -5 -10>, angularTarget=<0 0 0>}
      localJointOffset=<2 0 0>
     }]
    relativeSpaceBone {boneName=pelvis}
   }
   n64 = Get Not in SkyDiving() -> (Not in SkyDiving:bool)
   n65 = Anim Dynamics - Bone: dyn_belt_r_strap_01!(ComponentPose=@n63, bAlphaBoolEnabled=@n64, ExternalForce="0,0,0") -> (Pose) {
    alphaInputType=Bool
    angularDampingOverride=1
    angularSpringConstant=1000
    bAngularSpring=true
    bOverrideAngularDamping=true
    bUseGravityOverride=true
    bUsePlanarLimit=false
    simulationSpace=RootRelative
    alphaBoolBlend {blendInTime=0.1, blendOption=Linear, blendOutTime=0.1}
    boundBone {boneName=dyn_belt_r_strap_01}
    componentAppliedLinearAccClamp=<3000 3000 3000>
    componentLinearAccScale=<0.4 0.5 0.5>
    componentLinearVelScale=<0.5 0.5 0.5>
    gravityOverride {z=-980}
    physicsBodyDefinitions [{
      collisionType=CustomSphere
      sphereCollisionRadius=2.5
      boundBone {boneName=dyn_belt_r_strap_01}
      boxExtents=<8 8 8>
      constraintSetup {angularLimitsMax=<0 25 25>, angularLimitsMin=<0 5 0>, angularTarget=<0 0 0>}
      localJointOffset=<2 0 0>
     }]
    relativeSpaceBone {boneName=pelvis}
   }
   n66 = Anim Dynamics - Bone: dyn_belt_r_strap_02!(ComponentPose=@n65, bAlphaBoolEnabled=@n64, ExternalForce="0,0,0") -> (Pose) {
    alphaInputType=Bool
    angularDampingOverride=1
    angularSpringConstant=1000
    bAngularSpring=true
    bOverrideAngularDamping=true
    bUseGravityOverride=true
    bUsePlanarLimit=false
    simulationSpace=RootRelative
    alphaBoolBlend {blendInTime=0.1, blendOption=Linear, blendOutTime=0.1}
    boundBone {boneName=dyn_belt_r_strap_02}
    componentAppliedLinearAccClamp=<3000 3000 3000>
    componentLinearAccScale=<0.4 0.5 0.5>
    componentLinearVelScale=<0.5 0.5 0.5>
    gravityOverride {z=-980}
    physicsBodyDefinitions [{
      collisionType=CustomSphere
      sphereCollisionRadius=2.5
      boundBone {boneName=dyn_belt_r_strap_02}
      boxExtents=<8 8 8>
      constraintSetup {angularLimitsMax=<0 25 25>, angularLimitsMin=<0 -5 -25>, angularTarget=<0 0 0>}
      localJointOffset=<2 0 0>
     }]
    relativeSpaceBone {boneName=pelvis}
    sphericalLimits [{limitRadius=8, drivingBone {boneName=calf_twist_01_r}, sphereLocalOffset=<-7.2167 4.084 0>}]
   }
   n67 = Anim Dynamics - Bone: dyn_belt_r_strap_03!(ComponentPose=@n66, bAlphaBoolEnabled=@n64, ExternalForce="0,0,0") -> (Pose) {
    alphaInputType=Bool
    angularDampingOverride=1
    angularSpringConstant=1000
    bAngularSpring=true
    bOverrideAngularDamping=true
    bUseGravityOverride=true
    bUsePlanarLimit=false
    bUseSphericalLimits=true
    simulationSpace=RootRelative
    alphaBoolBlend {blendInTime=0.1, blendOption=Linear, blendOutTime=0.1}
    boundBone {boneName=dyn_belt_r_strap_03}
    componentAppliedLinearAccClamp=<3000 3000 3000>
    componentLinearAccScale=<0.4 0.5 0.5>
    componentLinearVelScale=<0.5 0.5 0.5>
    gravityOverride {z=-980}
    physicsBodyDefinitions [{
      collisionType=CustomSphere
      sphereCollisionRadius=2.5
      boundBone {boneName=dyn_belt_r_strap_03}
      boxExtents=<8 8 8>
      constraintSetup {angularLimitsMax=<0 25 25>, angularLimitsMin=<0 -5 -25>, angularTarget=<0 0 0>}
      localJointOffset=<2 0 0>
     }]
    relativeSpaceBone {boneName=pelvis}
    sphericalLimits [
     {limitRadius=20, drivingBone {boneName=calf_r}, sphereLocalOffset=<0 0 0>}
     {limitRadius=15, drivingBone {boneName=foot_r}, sphereLocalOffset=<0 0 0>}
     {limitRadius=0, drivingBone {boneName=None}, sphereLocalOffset=<0 0 0>}
    ]
   }
   n68 = Get bIsSkyDiving()
   n69 = Anim Dynamics - Bone: dyn_belt_r_strap_01(ComponentPose=@n67, bAlphaBoolEnabled=@n68, ExternalForce="0,0,0") {
    alphaInputType=Bool
    angularDampingOverride=1
    angularSpringConstant=1000
    bOverrideAngularDamping=true
    bUseGravityOverride=true
    bUsePlanarLimit=false
    simulationSpace=RootRelative
    alphaBoolBlend {blendInTime=0.1, blendOption=Linear, blendOutTime=0.1}
    boundBone {boneName=dyn_belt_r_strap_01}
    componentAppliedLinearAccClamp=<1000 1000 1000>
    componentLinearAccScale=<0.4 0.5 0.5>
    componentLinearVelScale=<0.5 0.5 0.5>
    gravityOverride {z=-500}
    physicsBodyDefinitions [{
      collisionType=CustomSphere
      sphereCollisionRadius=2.5
      boundBone {boneName=dyn_belt_r_strap_01}
      boxExtents=<8 8 8>
      constraintSetup {angularLimitsMax=<0 25 25>, angularLimitsMin=<0 5 0>, angularTarget=<0 0 0>}
      localJointOffset=<0.5 0 0>
     }]
    relativeSpaceBone {boneName=pelvis}
   }
   n70 = Anim Dynamics - Bone: dyn_belt_r_strap_02(ComponentPose=@n69, bAlphaBoolEnabled=@n68, ExternalForce="0,0,0") {
    alphaInputType=Bool
    angularDampingOverride=1
    angularSpringConstant=1000
    bOverrideAngularDamping=true
    bUseGravityOverride=true
    bUsePlanarLimit=false
    simulationSpace=RootRelative
    alphaBoolBlend {blendInTime=0.1, blendOption=Linear, blendOutTime=0.1}
    boundBone {boneName=dyn_belt_r_strap_02}
    componentAppliedLinearAccClamp=<1000 1000 1000>
    componentLinearAccScale=<0.4 0.5 0.5>
    componentLinearVelScale=<0.5 0.5 0.5>
    gravityOverride {z=-500}
    physicsBodyDefinitions [{
      collisionType=CustomSphere
      sphereCollisionRadius=2.5
      boundBone {boneName=dyn_belt_r_strap_02}
      boxExtents=<8 8 8>
      constraintSetup {angularLimitsMax=<0 25 25>, angularLimitsMin=<0 -5 -25>, angularTarget=<0 0 0>}
      localJointOffset=<0.5 0 0>
     }]
    relativeSpaceBone {boneName=pelvis}
    sphericalLimits [{limitRadius=8, drivingBone {boneName=calf_twist_01_r}, sphereLocalOffset=<-7.2167 4.084 0>}]
   }
   n71 = Anim Dynamics - Bone: dyn_belt_r_strap_03(ComponentPose=@n70, bAlphaBoolEnabled=@n68, ExternalForce="0,0,0") {
    alphaInputType=Bool
    angularDampingOverride=1
    angularSpringConstant=1000
    bOverrideAngularDamping=true
    bUseGravityOverride=true
    bUsePlanarLimit=false
    bUseSphericalLimits=true
    simulationSpace=RootRelative
    alphaBoolBlend {blendInTime=0.1, blendOption=Linear, blendOutTime=0.1}
    boundBone {boneName=dyn_belt_r_strap_03}
    componentAppliedLinearAccClamp=<1000 1000 1000>
    componentLinearAccScale=<0.4 0.5 0.5>
    componentLinearVelScale=<0.5 0.5 0.5>
    gravityOverride {z=-500}
    physicsBodyDefinitions [{
      collisionType=CustomSphere
      sphereCollisionRadius=2.5
      boundBone {boneName=dyn_belt_r_strap_03}
      boxExtents=<8 8 8>
      constraintSetup {angularLimitsMax=<0 25 25>, angularLimitsMin=<0 -5 -25>, angularTarget=<0 0 0>}
      localJointOffset=<0.5 0 0>
     }]
    relativeSpaceBone {boneName=pelvis}
    sphericalLimits [
     {limitRadius=20, drivingBone {boneName=calf_r}, sphereLocalOffset=<0 0 0>}
     {limitRadius=15, drivingBone {boneName=foot_r}, sphereLocalOffset=<0 0 0>}
     {limitRadius=0, drivingBone {boneName=None}, sphereLocalOffset=<0 0 0>}
    ]
   }
   Save cached pose 'PreSuspenderSetup'(Pose=Component To Local(ComponentPose=@n71)) {cacheName=PreSuspenderSetup}
   n72 = Get bIsJumping()
   n73 = Get bIsKneeSliding()
   n74 = Get Crouching_Not_sliding()
   n75 = Get Gravity() -> (Gravity:Vector)
   n76 = Get Not Crouching() -> (Not Crouching:bool)
   n77 = Anim Dynamics - Bone: dyn_suspenders_01_s!(ComponentPose=Local To Component(LocalPose=Use cached pose 'PreSuspenderSetup'()), bAlphaBoolEnabled:bool=@n76, ExternalForce=@n75) -> (Pose) {
    alphaInputType=Bool
    angularDampingOverride=1
    angularSpringConstant=2000
    bAngularSpring=true
    bOverrideAngularDamping=true
    bUseGravityOverride=true
    simulationSpace=BoneRelative
    alphaBoolBlend {blendInTime=0.3, blendOption=ExpInOut, blendOutTime=0.3}
    boundBone {boneName=dyn_suspenders_01_s}
    componentAppliedLinearAccClamp=<3000 3000 3000>
    componentLinearAccScale=<0.3 0.3 0.3>
    componentLinearVelScale=<0.3 0.3 0.3>
    gravityOverride {z=-980}
    physicsBodyDefinitions [{
      boundBone {boneName=dyn_suspenders_01_s}
      boxExtents=<20 20 20>
      constraintSetup {angularLimitsMax=<0 -5 15>, angularLimitsMin=<0 -20 -10>, angularTarget=<0 0 0>}
      localJointOffset=<8 0 0>
     }]
    relativeSpaceBone {boneName=root}
   }
   n78 = Anim Dynamics - Bone: dyn_suspenders_01_s2!(ComponentPose=@n77, bAlphaBoolEnabled:bool=@n76, ExternalForce=@n75) -> (Pose) {
    alphaInputType=Bool
    angularDampingOverride=1
    angularSpringConstant=2000
    bAngularSpring=true
    bOverrideAngularDamping=true
    bUseGravityOverride=true
    simulationSpace=BoneRelative
    alphaBoolBlend {blendInTime=0.3, blendOption=ExpInOut, blendOutTime=0.3}
    boundBone {boneName=dyn_suspenders_01_s2}
    componentAppliedLinearAccClamp=<3000 3000 3000>
    componentLinearAccScale=<0.3 0.3 0.3>
    componentLinearVelScale=<0.3 0.3 0.3>
    gravityOverride {z=-980}
    physicsBodyDefinitions [{
      boundBone {boneName=dyn_suspenders_01_s2}
      boxExtents=<15 15 15>
      constraintSetup {angularLimitsMax=<0 10 10>, angularLimitsMin=<0 -35 -10>, angularTarget=<0 0 0>}
      localJointOffset=<8 0 0>
     }]
    relativeSpaceBone {boneName=root}
   }
   n79 = Anim Dynamics - Bone: dyn_suspenders_01_b!(ComponentPose=@n78, bAlphaBoolEnabled:bool=@n76, ExternalForce="0,0,0") -> (Pose) {
    alphaInputType=Bool
    angularDampingOverride=1
    bOverrideAngularDamping=true
    bUseGravityOverride=true
    simulationSpace=BoneRelative
    alphaBoolBlend {blendInTime=0.3, blendOption=ExpInOut, blendOutTime=0.3}
    boundBone {boneName=dyn_suspenders_01_b}
    componentAppliedLinearAccClamp=<3000 3000 3000>
    componentLinearAccScale=<0.1 0.1 0.1>
    componentLinearVelScale=<0.1 0.1 0.1>
    gravityOverride {z=-980}
    physicsBodyDefinitions [{
      boundBone {boneName=dyn_suspenders_01_b}
      boxExtents=<15 15 15>
      constraintSetup {angularLimitsMax=<0 20 -5>, angularLimitsMin=<0 -30 -40>, angularTarget=<0 0 0>}
      localJointOffset=<8 0 0>
     }]
    relativeSpaceBone {boneName=root}
   }
   n80 = Anim Dynamics - Bone: dyn_suspenders_01_s(ComponentPose=@n79, bAlphaBoolEnabled=@n74, ExternalForce="0,0,0") {
    alphaInputType=Bool
    angularDampingOverride=1
    angularSpringConstant=2000
    bAngularSpring=true
    bOverrideAngularDamping=true
    bUseGravityOverride=true
    simulationSpace=RootRelative
    alphaBoolBlend {blendInTime=0.1, blendOption=Linear, blendOutTime=0.1}
    boundBone {boneName=dyn_suspenders_01_s}
    componentAppliedLinearAccClamp=<3000 3000 3000>
    componentLinearAccScale=<0.4 0.4 0.4>
    componentLinearVelScale=<0.5 0.5 0.5>
    gravityOverride {z=-980}
    physicsBodyDefinitions [{
      boundBone {boneName=dyn_suspenders_01_s}
      boxExtents=<20 10 10>
      constraintSetup {angularLimitsMax=<0 0 30>, angularLimitsMin=<0 -50 -30>, angularTarget=<0 0 0>}
      localJointOffset=<1 0 0>
     }]
    relativeSpaceBone {boneName=pelvis}
   }
   n81 = Anim Dynamics - Bone: dyn_suspenders_01_s2(ComponentPose=@n80, bAlphaBoolEnabled=@n74, ExternalForce="0,0,0") {
    alphaInputType=Bool
    angularDampingOverride=1
    angularSpringConstant=2000
    bAngularSpring=true
    bOverrideAngularDamping=true
    bUseGravityOverride=true
    simulationSpace=RootRelative
    alphaBoolBlend {blendInTime=0.1, blendOption=Linear, blendOutTime=0.1}
    boundBone {boneName=dyn_suspenders_01_s2}
    componentAppliedLinearAccClamp=<3000 3000 3000>
    componentLinearAccScale=<0.4 0.4 0.4>
    componentLinearVelScale=<0.5 0.5 0.5>
    gravityOverride {z=-980}
    physicsBodyDefinitions [{
      boundBone {boneName=dyn_suspenders_01_s2}
      boxExtents=<20 10 10>
      constraintSetup {angularLimitsMax=<0 -10 30>, angularLimitsMin=<0 -50 -5>, angularTarget=<0 0 0>}
      localJointOffset=<1 0 0>
     }]
    relativeSpaceBone {boneName=pelvis}
   }
   n82 = Anim Dynamics - Bone: dyn_suspenders_01_b(ComponentPose=@n81, bAlphaBoolEnabled=@n74, ExternalForce="0,0,0") {
    alphaInputType=Bool
    angularDampingOverride=1
    bOverrideAngularDamping=true
    bUseGravityOverride=true
    simulationSpace=RootRelative
    alphaBoolBlend {blendInTime=0.1, blendOption=Linear, blendOutTime=0.1}
    boundBone {boneName=dyn_suspenders_01_b}
    componentAppliedLinearAccClamp=<3000 3000 3000>
    componentLinearAccScale=<0.4 0.4 0.4>
    componentLinearVelScale=<0.5 0.5 0.5>
    gravityOverride {z=-980}
    physicsBodyDefinitions [{
      boundBone {boneName=dyn_suspenders_01_b}
      boxExtents=<10 10 10>
      constraintSetup {angularLimitsMax=<0 0 0>, angularLimitsMin=<0 -90 -30>, angularTarget=<0 0 0>}
      localJointOffset=<1 0 0>
     }]
    relativeSpaceBone {boneName=pelvis}
   }
   n83 = Anim Dynamics - Bone: dyn_suspenders_01_s(ComponentPose=@n82, bAlphaBoolEnabled=@n73, ExternalForce="0,0,0") {
    alphaInputType=Bool
    angularDampingOverride=1
    angularSpringConstant=2000
    bAngularSpring=true
    bOverrideAngularDamping=true
    bUseGravityOverride=true
    simulationSpace=RootRelative
    alphaBoolBlend {blendInTime=0.1, blendOption=Linear, blendOutTime=0.1}
    boundBone {boneName=dyn_suspenders_01_s}
    componentAppliedLinearAccClamp=<3000 3000 3000>
    componentLinearAccScale=<0.4 0.4 0.4>
    componentLinearVelScale=<0.5 0.5 0.5>
    gravityOverride {z=-980}
    physicsBodyDefinitions [{
      boundBone {boneName=dyn_suspenders_01_s}
      boxExtents=<20 10 10>
      constraintSetup {angularLimitsMax=<0 0 30>, angularLimitsMin=<0 -20 -30>, angularTarget=<0 0 0>}
      localJointOffset=<1 0 0>
     }]
    relativeSpaceBone {boneName=pelvis}
   }
   n84 = Anim Dynamics - Bone: dyn_suspenders_01_s2(ComponentPose=@n83, bAlphaBoolEnabled=@n73, ExternalForce="0,0,0") {
    alphaInputType=Bool
    angularDampingOverride=1
    angularSpringConstant=2000
    bAngularSpring=true
    bOverrideAngularDamping=true
    bUseGravityOverride=true
    simulationSpace=RootRelative
    alphaBoolBlend {blendInTime=0.1, blendOption=Linear, blendOutTime=0.1}
    boundBone {boneName=dyn_suspenders_01_s2}
    componentAppliedLinearAccClamp=<3000 3000 3000>
    componentLinearAccScale=<0.4 0.4 0.4>
    componentLinearVelScale=<0.5 0.5 0.5>
    gravityOverride {z=-980}
    physicsBodyDefinitions [{
      boundBone {boneName=dyn_suspenders_01_s2}
      boxExtents=<20 10 10>
      constraintSetup {angularLimitsMax=<0 0 30>, angularLimitsMin=<0 -20 -5>, angularTarget=<0 0 0>}
      localJointOffset=<1 0 0>
     }]
    relativeSpaceBone {boneName=pelvis}
   }
   n85 = "Transform (Modify) Bone - Bone: dyn_suspenders_01_s2_MSTR"!(ComponentPose=@n84, bAlphaBoolEnabled:bool=@n72, Translation="0,0,0", Rotation="4.981,-5.019,-0.437", Scale="1,1,1") -> (Pose) {
    alphaInputType=Bool
    rotationMode=BMM_Additive
    rotationSpace=BCS_BoneSpace
    translationMode=BMM_Additive
    alphaBoolBlend {blendInTime=0.3, blendOutTime=0.3}
    boneToModify {boneName=dyn_suspenders_01_s2_MSTR}
    rotation=<rot 4.981 -5.019 -0.437>
   }
   n86 = "Transform (Modify) Bone - Bone: dyn_suspenders_01_b_MSTR"!(ComponentPose=@n85, bAlphaBoolEnabled:bool=@n72, Translation="0,0,0", Rotation="0,-10,0", Scale="1,1,1") -> (Pose) {
    alphaInputType=Bool
    rotationMode=BMM_Additive
    rotationSpace=BCS_BoneSpace
    translationMode=BMM_Additive
    alphaBoolBlend {blendInTime=0.3, blendOutTime=0.3}
    boneToModify {boneName=dyn_suspenders_01_b_MSTR}
    rotation {yaw=-10}
   }
   n87 = Anim Dynamics - Bone: dyn_suspenders_01_s(ComponentPose=Local To Component(LocalPose=Use cached pose 'PreSuspenderSetup'()), bAlphaBoolEnabled=True, ExternalForce="0,0,0") {
    alphaInputType=Bool
    angularDampingOverride=1
    angularSpringConstant=2000
    bAngularSpring=true
    bOverrideAngularDamping=true
    bUseGravityOverride=true
    simulationSpace=BoneRelative
    alphaBoolBlend {blendInTime=0.3, blendOption=ExpInOut, blendOutTime=0.3}
    boundBone {boneName=dyn_suspenders_01_s}
    componentAppliedLinearAccClamp=<3000 3000 3000>
    componentLinearAccScale=<0.1 0.1 0.1>
    componentLinearVelScale=<0.1 0.1 0.1>
    gravityOverride {z=-400}
    physicsBodyDefinitions [{
      boundBone {boneName=dyn_suspenders_01_s}
      boxExtents=<20 20 20>
      constraintSetup {angularLimitsMax=<0 -5 10>, angularLimitsMin=<0 -30 -10>, angularTarget=<0 0 0>}
      localJointOffset=<8 0 0>
     }]
    planarLimits [{drivingBone {boneName=thigh_l}, planeTransform {location=<0 0 -4>, rotation=<rot 10 0 180>}}]
    relativeSpaceBone {boneName=root}
   }
   n88 = Anim Dynamics - Bone: dyn_suspenders_01_s2(ComponentPose=@n87, bAlphaBoolEnabled=True, ExternalForce="0,0,0") {
    alphaInputType=Bool
    angularDampingOverride=1
    angularSpringConstant=2000
    bAngularSpring=true
    bOverrideAngularDamping=true
    bUseGravityOverride=true
    simulationSpace=BoneRelative
    alphaBoolBlend {blendInTime=0.3, blendOption=ExpInOut, blendOutTime=0.3}
    boundBone {boneName=dyn_suspenders_01_s2}
    componentAppliedLinearAccClamp=<3000 3000 3000>
    componentLinearAccScale=<0.1 0.1 0.1>
    componentLinearVelScale=<0.1 0.1 0.1>
    gravityOverride {z=-400}
    physicsBodyDefinitions [{
      boundBone {boneName=dyn_suspenders_01_s2}
      boxExtents=<20 20 20>
      constraintSetup {angularLimitsMax=<0 -5 -5>, angularLimitsMin=<0 -30 -10>, angularTarget=<0 0 0>}
      localJointOffset=<8 0 0>
     }]
    planarLimits [{drivingBone {boneName=thigh_l}, planeTransform {location=<0 0 -4>, rotation=<rot 10 0 180>}}]
    relativeSpaceBone {boneName=root}
   }
   n89 = Anim Dynamics - Bone: dyn_suspenders_01_b(ComponentPose=@n88, bAlphaBoolEnabled=True, ExternalForce="0,0,0") {
    alphaInputType=Bool
    angularDampingOverride=1
    bOverrideAngularDamping=true
    bUseGravityOverride=true
    simulationSpace=BoneRelative
    alphaBoolBlend {blendInTime=0.3, blendOption=ExpInOut, blendOutTime=0.3}
    boundBone {boneName=dyn_suspenders_01_b}
    componentAppliedLinearAccClamp=<3000 3000 3000>
    componentLinearAccScale=<0.1 0.1 0.1>
    componentLinearVelScale=<0.1 0.1 0.1>
    gravityOverride {z=-400}
    physicsBodyDefinitions [{
      boundBone {boneName=dyn_suspenders_01_b}
      boxExtents=<20 20 20>
      constraintSetup {angularLimitsMax=<0 20 -10>, angularLimitsMin=<0 -20 -30>, angularTarget=<0 0 0>}
      localJointOffset=<8 0 0>
     }]
    relativeSpaceBone {boneName=root}
   }
   n90 = "Transform (Modify) Bone - Bone: dyn_suspenders_01_b_MSTR"!(ComponentPose=@n89, Alpha:float=1, Translation="0,0,0", Rotation="0.000,0.000,15.000", Scale="1,1,1") -> (Pose) {
    rotationMode=BMM_Additive
    translationMode=BMM_Additive
    alphaBoolBlend {blendOption=Linear}
    boneToModify {boneName=dyn_suspenders_01_b_MSTR}
    rotation {roll=15}
   }
   n91 = Control Rig!(
    Source=Blend Poses by bool(BlendPose_0=Component To Local(ComponentPose=@n90), BlendPose_1=Component To Local(ComponentPose=@n86), bActiveValue=Get bIsSkyDiving()) {blendType=Linear}
    Alpha
    Emote:double=Get Emote()
    Sprint:double=Get Sprint() -> (Sprint:double)
    Crouching_moving:double=Get Crouching Moving()
    StandingStill:double=Get StandingStill_CTRLRig_Alpha() -> (StandingStill_CTRLRig_Alpha:double)
    Skydiving:double=Get IsSkydiving_CTRLRig_Alpha() -> (IsSkydiving_CTRLRig_Alpha:double)
    Jumping:double=Get Jumping_CTRLRig_Alpha() -> (Jumping_CTRLRig_Alpha:double)
    Parachuting:double=Get Parachuting_CTRLRig_Alpha() -> (Parachuting_CTRLRig_Alpha:double)
   ) -> (Pose) {
    alphaBoolBlend {blendOption=Linear}
    controlRigAssetReference {blueprintRigClass=&/F_MED_TrafficHat_Boba_CtrlRig.F_MED_TrafficHat_Boba_CtrlRig_C}
   }
   Save cached pose 'BRSetup'(Pose=@n91) {cacheName=BRSetup}
  ]}
 EventGraph {nodes [
   Event OnCharacterCustomizationLoaded()
   > n1 = Set Skel Mesh(Skel Mesh=GetPartSkeletalMeshComponent())
   > n2 = Get Dynamic Material Instance(SkeletalMeshComponent=@n1)
   > Set Mat Instance Dynamic(Mat Instance Dynamic=@n2)
   > n3 = GetOrCreateDynamicOverlayMaterial(MeshComponent=@n1, MaterialName=None)
   > Set Overlay Mat Instance Dynamic(Overlay Mat Instance Dynamic=@n3)
   Event BlueprintInitializeAnimation()
   > Set Skel Mesh(Skel Mesh=GetOwnerSkeletalMeshComponent())
   n4 = TryGetPawnOwner()
   > Is Valid(InputObject=@n4)
   > n5 = Cast To FortPlayerPawnAthena(Object=@n4)
   > Set PlayerPawn(PlayerPawn:Fort Player Pawn Athena Object Reference=@n5) -> (Output_Get:Fort Player Pawn Athena Object Reference)
   n6 = Event BlueprintUpdateAnimation()
   > Set Delta Time X(Delta Time X=@n6.DeltaTimeX)
   > n7 = Sequence!(execute) -> (then_0, then_1, then_2:Exec, then_3:Exec)
   Is Valid(exec=@n7.then_3, InputObject=Get PlayerPawn() -> (PlayerPawn:Fort Player Pawn Athena Object Reference))
   > Set WPO Globals Function(
    Material=Get Mat Instance Dynamic()
    Overlay Material=Get Overlay Mat Instance Dynamic()
    Skeletal Mesh=Get Skel Mesh()
    "Local Noise Transform  Location"="Get Local Noise Transform  Location"() -> ("Local Noise Transform  Location":Vector)
    Local Noise Offset=Get Local Noise Offset()
   )
   > Set WPO Regional Function!(
    execute
    Material:Material Instance Dynamic Object Reference=Get Mat Instance Dynamic()
    Overlay Material:Material Instance Dynamic Object Reference=Get Overlay Mat Instance Dynamic()
    Skeletal Mesh:Skeletal Mesh Component Object Reference=Get Skel Mesh()
    Velocity Transform:Name=foot_r
    Velocity Interp Speed:double=Get Velocity Interp Speed()
    Collision Transform:Name=calf_r
    Collision Transform Axis:EAxis Enum
    Invert Collision Transform Axis:bool=false
    Delta Time X:double=Get Delta Time X()
    Previous Velocity Transform Location:"Vector (by ref)"=Get R1 Velocity Transform Location()
    Previous Velocity Transform Velocity:"Vector (by ref)"=Get R1 Velocity Transform Velocity()
    Region:int=1
   ) -> (then)
   > SetScalarParameterValue(self=Get Mat Instance Dynamic(), ParameterName=ClothSim_WindStrength, Value=SelectFloat(A=1, B=0, bPickA=Get bIsSkydiveFloating()))
   n8 = Get bIsPlayingEmote()
   Set Emote(
    execute=@n7.then_2
    Emote:double=FInterpTo(Current=Get Emote(), Target="To Float (Boolean)"(InBool=@n8), DeltaTime=Get Delta Time X(), InterpSpeed=10)
   ) -> (Output_Get:double)
   n9 = AND Boolean(A=Get bIsCrouching(), B=NOT Boolean(A=Get bIsSliding()))
   > Set Crouching Moving(Crouching Moving:double=FInterpTo(Current=Get Crouching Moving(), Target="To Float (Boolean)"(InBool=@n9), DeltaTime=Get Delta Time X(), InterpSpeed=10)) -> (Output_Get:double)
   n10 = Get bIsSprinting()
   > Set Sprint(Sprint:double=FInterpTo(Current=Get Sprint(), Target="To Float (Boolean)"(InBool=@n10), DeltaTime=Get Delta Time X(), InterpSpeed=10)) -> (Output_Get:double)
   n11 = Get bIsSliding()
   > Set Crouching_Not_sliding(Crouching_Not_sliding:bool=AND Boolean(A=Get bIsCrouching(), B=NOT Boolean(A=@n11))) -> (Output_Get:bool)
   > Set Not in SkyDiving(Not in SkyDiving:bool=NOT Boolean(A=Get bIsSkyDiving())) -> (Output_Get:bool)
   > Set Crouching_backpedaling(Crouching_backpedaling:bool=AND Boolean(A=Get bIsCrouching(), B=Get bIsBackpedaling() -> (bIsBackpedaling:bool))) -> (Output_Get:bool)
   > n12 = Set Not Crouching(Not Crouching:bool=NOT Boolean(A=Get bIsCrouching())) -> (Output_Get:bool)
   n13 = Get bIsMoving2D()
   > n14 = Set StandingStill(StandingStill:bool=AND Boolean(A=@n12, B=NOT Boolean(A=@n13))) -> (Output_Get:bool)
   > Set StandingStill_CTRLRig_Alpha(StandingStill_CTRLRig_Alpha:double=FInterpTo(Current=Get StandingStill_CTRLRig_Alpha(), Target="To Float (Boolean)"(InBool=@n14), DeltaTime=Get Delta Time X(), InterpSpeed=10)) -> (Output_Get:double)
   n15 = Get bIsSkyDiving()
   > Set IsSkydiving_CTRLRig_Alpha(IsSkydiving_CTRLRig_Alpha:double=FInterpTo(Current=Get IsSkydiving_CTRLRig_Alpha(), Target="To Float (Boolean)"(InBool=@n15), DeltaTime=Get Delta Time X(), InterpSpeed=10)) -> (Output_Get:double)
   n16 = Get bIsCrouching()
   n17 = Get bIsSliding()
   > Set LookAt_Alpha(LookAt_Alpha:double=SelectFloat(A=1, B=0.7, bPickA=OR Boolean(A=@n16, B=@n17))) -> (Output_Get:double)
   n18 = OR Boolean(A=Get bIsJumping(), B=Get bIsFalling())
   > Set Jumping_CTRLRig_Alpha(Jumping_CTRLRig_Alpha:double=FInterpTo(Current=Get Jumping_CTRLRig_Alpha(), Target="To Float (Boolean)"(InBool=@n18), DeltaTime=Get Delta Time X(), InterpSpeed=20)) -> (Output_Get:double)
   n19 = Get bIsParachuteOpened()
   > Set Parachuting_CTRLRig_Alpha(Parachuting_CTRLRig_Alpha:double=FInterpTo(Current=Get Parachuting_CTRLRig_Alpha(), Target="To Float (Boolean)"(InBool=@n19), DeltaTime=Get Delta Time X(), InterpSpeed=20)) -> (Output_Get:double)
   n20 = Get bIsPlayingEmote()
   > Set IsNotPlayingEmote(IsNotPlayingEmote:double="To Float (Boolean)"(InBool=NOT Boolean(A=@n20))) -> (Output_Get:double)
   n21 = Gravity Override State Function DEPRECATED(
    execute=@n7.then_1
    CharacterPartAnimInstance=Self-Reference()
    GameTimeInSeconds=GetGameTimeInSeconds()
    GravityOverrideParams=Get Gravity Override Params()
    RBAN Params=Get RBAN Params()
   )
   n22 = Get bIsJumping()
   n23 = Get bIsFalling()
   > Set Gravity(Gravity:Vector=SelectVector(A="0,0,100", B=@n21, bPickA=OR Boolean(A=@n22, B=@n23))) -> (Output_Get:Vector)
   OR Boolean(A=false, B=Get bIsSprinting())
   AND Boolean(A=@n10, B=NOT Boolean(A=Get bIsCrouching()))
   Get bIsMoving2D()
  ]}
]
variables [
 Crouching Moving {type=number}
 Crouching_backpedaling {type=boolean}
 Crouching_Not_sliding {type=boolean}
 Delta Time X {type=number}
 Emote {type=number}
 Gravity {type=Vector}
 Gravity Override Params {type=GravityOverrideParamsStruct, default {onGround_Moving {gravityOverride=<0 0 -250>}}}
 Gravity_Skydive {type=Vector}
 IsNotPlayingEmote {type=number}
 IsSkydiving_CTRLRig_Alpha {type=number}
 Jumping_CTRLRig_Alpha {type=number}
 Local Noise Offset {type=Vector}
 "Local Noise Transform  Location" {type=Vector}
 LookAt_Alpha {type=number}
 Mat Instance Dynamic {type=MaterialInstanceDynamic}
 Not Crouching {type=boolean}
 Not in SkyDiving {type=boolean}
 Overlay Mat Instance Dynamic {type=MaterialInstanceDynamic}
 Parachuting_CTRLRig_Alpha {type=number}
 PlayerPawn {type=FortPlayerPawnAthena}
 R1 Velocity Transform Location {type=Vector}
 R1 Velocity Transform Velocity {type=Vector}
 RBAN Params {
  type=RBANParamsStruct
  default {
   dBNO {simSpaceSettings {
     maxAngularAcceleration=1000
     maxAngularVelocity=3
     maxLinearAcceleration=300
     maxLinearVelocity=700
     velocityScaleZ=0
     externalLinearDragV=<0 0.5 0>
    }}
   dBNO_Carried {simSpaceSettings {
     maxAngularAcceleration=1000
     maxAngularVelocity=3
     maxLinearAcceleration=300
     maxLinearVelocity=700
     velocityScaleZ=0
     externalLinearDragV=<0 0.5 0>
    }}
   emoteOrMelee {simSpaceSettings {
     maxAngularAcceleration=1000
     maxAngularVelocity=3
     maxLinearAcceleration=300
     maxLinearVelocity=700
     velocityScaleZ=0
     externalLinearDragV=<0 0.5 0>
    }}
   front_End {simSpaceSettings {
     maxAngularAcceleration=1000
     maxAngularVelocity=3
     maxLinearAcceleration=300
     maxLinearVelocity=700
     velocityScaleZ=0
     externalLinearDragV=<0 0.5 0>
    }}
   onGround_Crouch_Moving {simSpaceSettings {
     maxAngularAcceleration=1000
     maxAngularVelocity=3
     maxLinearAcceleration=300
     maxLinearVelocity=700
     velocityScaleZ=0
     externalLinearDragV=<0 0.5 0>
    }}
   onGround_Falling {simSpaceSettings {
     maxAngularAcceleration=1000
     maxAngularVelocity=3
     maxLinearAcceleration=300
     maxLinearVelocity=700
     velocityScaleZ=0
     externalLinearDragV=<0 0.5 0>
    }}
   onGround_Moving {simSpaceSettings {
     maxAngularAcceleration=1000
     maxAngularVelocity=3
     maxLinearAcceleration=300
     maxLinearVelocity=700
     velocityScaleZ=0
     externalLinearDragV=<0 0.5 0>
    }}
   onGround_Standing {simSpaceSettings {
     maxAngularAcceleration=1000
     maxAngularVelocity=3
     maxLinearAcceleration=300
     maxLinearVelocity=700
     velocityScaleZ=0
     externalLinearDragV=<0 0.5 0>
    }}
   rail_Grinding {simSpaceSettings {
     maxAngularAcceleration=1000
     maxAngularVelocity=3
     maxLinearAcceleration=300
     maxLinearVelocity=700
     velocityScaleZ=0
     worldAlpha=0.1
     externalLinearDragV=<0 0.5 0>
    }}
   riding_Moving {simSpaceSettings {
     maxAngularAcceleration=1000
     maxAngularVelocity=3
     maxLinearAcceleration=300
     maxLinearVelocity=700
     velocityScaleZ=0
     externalLinearDragV=<0 0.5 0>
    }}
   riding_Moving_Targeting {simSpaceSettings {
     maxAngularAcceleration=1000
     maxAngularVelocity=3
     maxLinearAcceleration=300
     maxLinearVelocity=700
     velocityScaleZ=0
     externalLinearDragV=<0 0.5 0>
    }}
   riding_Standing {simSpaceSettings {
     maxAngularAcceleration=1000
     maxAngularVelocity=3
     maxLinearAcceleration=300
     maxLinearVelocity=700
     velocityScaleZ=0
     externalLinearDragV=<0 0.5 0>
    }}
   skydive_DivingDown {simSpaceSettings {
     maxAngularAcceleration=1000
     maxAngularVelocity=3
     maxLinearAcceleration=300
     maxLinearVelocity=700
     velocityScaleZ=0
     externalLinearDragV=<0 0.5 0>
    }}
   skydive_FloatingDown {
    componentLinearAccScale=<0 0.5 0>
    componentLinearVelScale=<0 0.5 0>
    simSpaceSettings {
     maxAngularAcceleration=1000
     maxAngularVelocity=3
     maxLinearAcceleration=300
     maxLinearVelocity=700
     velocityScaleZ=0
     externalLinearDragV=<0 0.5 0>
    }
   }
   skydive_GoingUp {simSpaceSettings {
     maxAngularAcceleration=1000
     maxAngularVelocity=3
     maxLinearAcceleration=300
     maxLinearVelocity=700
     velocityScaleZ=0
     externalLinearDragV=<0 0.5 0>
    }}
   skydive_Parachuting {
    componentLinearAccScale=<0 0.1 0.3>
    componentLinearVelScale=<0 0.1 0.3>
    simSpaceSettings {
     maxAngularAcceleration=1000
     maxAngularVelocity=3
     maxLinearAcceleration=300
     maxLinearVelocity=700
     velocityScaleZ=0
     externalLinearDragV=<0 0.5 0>
    }
   }
   skydive_Surfing {simSpaceSettings {
     maxAngularAcceleration=1000
     maxAngularVelocity=3
     maxLinearAcceleration=300
     maxLinearVelocity=700
     velocityScaleZ=0
     externalLinearDragV=<0 0.5 0>
    }}
   swimming {simSpaceSettings {
     maxAngularAcceleration=1000
     maxAngularVelocity=3
     maxLinearAcceleration=300
     maxLinearVelocity=700
     velocityScaleZ=0
     externalLinearDragV=<0 0.5 0>
    }}
   tactical_Sprint {simSpaceSettings {
     maxAngularAcceleration=1000
     maxAngularVelocity=3
     maxLinearAcceleration=300
     maxLinearVelocity=700
     velocityScaleZ=0
     externalLinearDragV=<0 0.5 0>
    }}
   ziplining {simSpaceSettings {
     maxAngularAcceleration=1000
     maxAngularVelocity=3
     maxLinearAcceleration=300
     maxLinearVelocity=700
     velocityScaleZ=0
     externalLinearDragV=<0 0.5 0>
    }}
  }
 }
 Skel Mesh {type=SkeletalMeshComponent}
 Sprint {type=number}
 StandingStill {type=boolean}
 StandingStill_CTRLRig_Alpha {type=number}
 suspenders_component_linear {type=Vector}
 Velocity Interp Speed {type=number}
]
